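# Unlit translucent master material for Niagara ribbon trails, with tunable color, opacity and scrolling textures.
# Scaled, time-panned UV1 with an optional axis swizzle feeds both color and opacity textures.
# A selectable color channel blends low/high colors, tinted by particle color through HDR and emissive adjustment functions.
# Opacity combines a selectable mask channel, particle alpha, squared front/back fades along UV0 U, and depth fade, skipped on mobile.
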Material /Game/Effects/Niagara/MasterMaterials/Ribbon/M_AnimTrailMaster
graph {
 properties {bUsedWithNiagaraRibbons=true, blendMode=BLEND_Translucent, pixelDepthOffsetMode=PDOM_Legacy, shadingModel=MSM_Unlit, usedShadingModels=Unlit}
 nodes [
  n1 = ChannelMaskParameter(None=TextureSampleParameter2D(UVs=NamedRerouteUsage()) {parameterName=Color Variation Texture, texture=/Game/Effects/Fort_Effects/Textures/Smoke/T_Smoke_Trail_Soft_Chromatic}.RGBA) {maskChannel=Green, parameterName=Color Texture Channel, defaultValue=<rgba 0 1 0 0>}
  n2 = ScalarParameter() {defaultValue=2, parameterName=Color Channel Pow}
  n3 = Multiply(
   A=ParticleColor().RGB
   B=LinearInterpolate(
    A=VectorParameter() {parameterName=Color Low}.RGB
    B=VectorParameter() {parameterName=Color High, defaultValue=<rgba 1 1 1 0>}.RGB
    Alpha=Power(Base=@n1.R, Exp=@n2)
   )
  )
  MaterialFunctionCall(Emissive=MaterialFunctionCall(LDR Input=Clamp(None=@n3), HDR Tint=@n3) {materialFunction=&DeriveHDRfromLDR}.RemappedRange) {materialFunction=&MF_FX_EmissiveAdjustments}
  n4 = NamedRerouteUsage()
  n5 = Power(
   Base=ChannelMaskParameter(None=TextureSampleParameter2D(UVs=@n4) {parameterName=Opacity Texture, texture=/Game/Effects/Fort_Effects/Textures/Tile/T_Trail_Masks_Tiled}.RGBA) {maskChannel=Green, parameterName=Opacity Texture Channel, defaultValue=<rgba 0 1 0 0>}.R
   Exp=ScalarParameter() {defaultValue=1, parameterName=Opacity Channel Pow}
  )
  n6 = DepthFade(Opacity=@n5, FadeDistance=ScalarParameter() {defaultValue=100, parameterName=DepthFade-Standard_Distance}) {fadeDistanceDefault=20}
  n7 = TextureCoordinate()
  n8 = Saturate(None=Multiply(A=ComponentMask(None=@n7) {r=true}, B=ScalarParameter() {defaultValue=8, parameterName=Front End Opacity Str}) {constB=8})
  n9 = ComponentMask(None=TextureCoordinate()) {r=true}
  n10 = Saturate(None=Multiply(A=OneMinus(None=@n9), B=ScalarParameter() {defaultValue=1, parameterName=Back End Opacity Str}))
  Multiply(
   A=Multiply(A=ParticleColor().A, B=ShadingPathSwitch(Default=@n6, Deferred=@n6, Mobile=@n5) {desc=Migrated from FeatureLevelSwitch})
   B=Multiply(A=Multiply(A=@n8, B=@n8), B=Multiply(A=@n10, B=@n10))
  )
  n11 = ScalarParameter() {defaultValue=1, parameterName=Overall Scale U}
  n12 = ScalarParameter() {defaultValue=1, parameterName=Overall Scale V}
  n13 = ScalarParameter() {parameterName=Overall Pan U}
  n14 = ScalarParameter() {parameterName=Overall Pan V}
  n15 = Add(
   A=Multiply(A=TextureCoordinate() {coordinateIndex=1}, B=AppendVector(A=@n11, B=@n12))
   B=Multiply(A=Time() {bOverride_Period=true, period=10000}, B=AppendVector(A=@n13, B=@n14))
  )
  NamedRerouteDeclaration(None=StaticSwitchParameter(True=MaterialFunctionCall(XY=@n15) {materialFunction=&Swizzle}.YX, False=@n15) {parameterName=Swizzle Textures?}) {name=Base UVs, nodeColor=<rgba 0.0078 0 1 1>}
 ]
}
