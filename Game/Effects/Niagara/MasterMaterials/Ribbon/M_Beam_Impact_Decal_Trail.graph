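# Unlit translucent master material for Niagara ribbons and mesh particles, named as a beam impact decal trail.
# A powered trail texture is multiplied by noise erosion driven by particle alpha and a U-edge fade.
# An inverted depth fade confines it to nearby intersecting geometry for a decal look, alongside a camera offset scaled from the fade distance.
# The mask tints particle color, and an alpha-driven divide sharpens its edges.
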Material /Game/Effects/Niagara/MasterMaterials/Ribbon/M_Beam_Impact_Decal_Trail
graph {
 properties {
  bUsedWithNiagaraMeshParticles=true
  bUsedWithNiagaraRibbons=true
  blendMode=BLEND_Translucent
  pixelDepthOffsetMode=PDOM_Legacy
  shadingModel=MSM_Unlit
  usedShadingModels=Unlit
 }
 nodes [
  Constant3Vector() {constant=<rgba 1 0 0 0>}
  Constant()
  DynamicParameter() {paramNames [UOffset,Param2,Param3,Param4]}
  n1 = TextureSampleParameter2D(UVs=NamedRerouteUsage()) {parameterName=Normal Texture, samplerType=SAMPLERTYPE_Normal, texture=/Game/Effects/Fort_Effects/Textures/Smoke/T_Smoke_Trail_Soft_N}
  n2 = NamedRerouteUsage()
  MaterialFunctionCall(
   BaseNormal=@n1.RGB
   AdditionalNormal=MaterialFunctionCall(
    Normal=TextureSampleParameter2D(UVs=@n2) {parameterName=Eataway Noise Normal, samplerType=SAMPLERTYPE_Normal, texture=/Game/Global/Textures/Default/Noise/T_Bubble_B_N}.RGB
    Flatness=ScalarParameter() {defaultValue=0.5, parameterName=Eataway Normal Flatness}
   ) {materialFunction=&FlattenNormal}.Result
  ) {materialFunction=&BlendAngleCorrectedNormals}
  n3 = ScalarParameter() {defaultValue=10, parameterName=Depth Fade Distance}
  MaterialFunctionCall(Offset Amount=Multiply(A=@n3) {constB=0.2}) {materialFunction=/Engine/Functions/Engine_MaterialFunctions02/WorldPositionOffset/CameraOffset}
  n4 = ParticleColor()
  n5 = TextureSampleParameter2D(UVs=NamedRerouteUsage()) {parameterName=EatawayNoise, samplerType=SAMPLERTYPE_LinearGrayscale, texture=/Game/Global/Textures/Default/Noise/T_Bubble_B}
  n6 = Saturate(None=Add(
    A=LinearInterpolate(Alpha=@n4.A) {constA=-1, constB=0}
    B=ChannelMaskParameter(None=@n5.RGBA) {maskChannel=Green, parameterName=Eataway Channel, defaultValue=<rgba 0 1 0 0>}.R
   ))
  n7 = TextureSampleParameter2D(UVs=NamedRerouteUsage()) {parameterName=Trail Texture, samplerType=SAMPLERTYPE_Grayscale, texture=/Game/Effects/Fort_Effects/Textures/Smoke/T_Smoke_Trail_Soft_NoClamp}
  n8 = ChannelMaskParameter(None=@n7.RGBA) {parameterName=Trail Channel}
  n9 = ScalarParameter() {defaultValue=4, parameterName=TrailPower}
  n10 = OneMinus(None=DepthFade(FadeDistance=@n3) {fadeDistanceDefault=2})
  n11 = ConstantBiasScale(None=ComponentMask(None=TextureCoordinate()) {r=true}) {bias=-0.5, scale=2}
  n12 = OneMinus(None=Power(Base=Abs(None=@n11)) {constExponent=20})
  n13 = Multiply(
   A=Multiply(A=NamedRerouteDeclaration(None=@n6) {name=Eataway, nodeColor=<rgba 0.5937 1 0 1>}, B=Power(Base=@n8.R, Exp=@n9))
   B=Multiply(A=Saturate(None=@n10), B=Saturate(None=@n12))
  ) {constA=1}
  Multiply(A=ParticleColor().RGB, B=@n13)
  NamedRerouteDeclaration(None=MaterialFunctionCall!(
    UV=TextureCoordinate() {coordinateIndex=1}
    Tile U=ScalarParameter() {defaultValue=1, parameterName=Tile U}
    Tile V
    Pan U
    Pan V
    Offset U
    Offset V
    Time
   ) {materialFunction=/Game/Effects/Fort_Effects/Materials/_Functions/MF_PanTileOffset}.UV's) {name="Trail UVs (UV1)", nodeColor=<rgba 0 1 0.5078 1>}
  NamedRerouteDeclaration(None=MaterialFunctionCall(
    UV=TextureCoordinate() {coordinateIndex=1}
    Tile U=ScalarParameter() {defaultValue=1, parameterName=Eataway Tile U}
    Tile V=ScalarParameter() {defaultValue=1, parameterName=Eataway Tile V}
   ) {materialFunction=&MF_PanTileOffset}.UV's) {name=Eataway UVs, nodeColor=<rgba 1 0.6328 0 1>}
  n14 = ScalarParameter() {defaultValue=0.1, parameterName=Edge Hardness}
  n15 = Power(Base=ParticleColor().A) {constExponent=0.5}
  Saturate(None=Divide(A=@n13, B=LinearInterpolate(B=@n14, Alpha=@n15) {constA=1}))
  ScalarParameter() {parameterName=Pan U}
  ScalarParameter() {defaultValue=0.6, parameterName=Metallic}
  ScalarParameter() {defaultValue=0.02, parameterName=Diffuse Color}
  ScalarParameter() {defaultValue=0.5, parameterName=Specular}
  ScalarParameter() {defaultValue=1, parameterName=Roughness}
 ]
}
