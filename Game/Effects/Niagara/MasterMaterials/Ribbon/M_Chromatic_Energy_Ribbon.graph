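# Unlit, translucent, two-sided master material for Niagara ribbon and mesh-particle energy effects, with a UseSpline switch for UV layout and world-position-offset thickening.
# Two panning chromatic textures form the energy, the secondary warping both main UVs and soft edge mask.
# Particle or gradient color tints it beneath a traveling two-color pulse.
# Dynamic parameters drive erosion, tiling, panning, width and pulse; opacity adds depth and optional view-angle fades.
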Material /Game/Effects/Niagara/MasterMaterials/Ribbon/M_Chromatic_Energy_Ribbon
graph {
 properties {
  bUsedWithNiagaraMeshParticles=true
  bUsedWithNiagaraRibbons=true
  blendMode=BLEND_Translucent
  pixelDepthOffsetMode=PDOM_Legacy
  shadingModel=MSM_Unlit
  translucencyLightingMode=TLM_VolumetricPerVertexDirectional
  twoSided=true
  usedShadingModels=Unlit
 }
 nodes [
  n1 = NamedRerouteUsage()
  n2 = Subtract(A=WorldPosition().XYZ, B=ParticlePositionWS())
  n3 = NamedRerouteUsage()
  n4 = DynamicParameter() {parameterIndex=1, defaultValue=<rgba 0 1 1 0>, paramNames [SplineWidthTip, YTileMult, PulseMaskTravel, PulseColorShift]}
  n5 = Multiply(A=@n3, B=@n4.SplineWidthTip)
  Add(
   A=Multiply(A=Multiply(A=@n1, B=@n2)) {constB=0.5}
   B=StaticSwitchParameter(
    True=MaterialFunctionCall(WidthBase=@n3, WidthTip=@n5) {materialFunction=&SplineThicken}.WorldPositionOffset
    False=Constant()
   ) {parameterName=UseSpline}
  )
  n6 = ParticleColor()
  n7 = TextureCoordinate()
  n8 = OneMinus(None=ComponentMask(None=@n7) {r=true})
  n9 = ScalarParameter() {defaultValue=4, parameterName=EndFadeHardness}
  n10 = StaticSwitchParameter(
   True=SphereMask(A=ComponentMask(None=@n7) {g=true}, B=Constant() {r=0.5}) {attenuationRadius=0.5, hardnessPercent=0}
   False=Saturate(None=Multiply(A=@n8, B=@n9) {constB=3})
  ) {parameterName=UseSpline}
  n11 = NamedRerouteUsage()
  n12 = Constant2Vector() {r=-0.5}
  n13 = StaticSwitchParameter(
   True=MaterialFunctionCall(XY=TextureCoordinate()) {materialFunction=&Swizzle}.YX
   False=TextureCoordinate() {coordinateIndex=1}
  ) {parameterName=UseSpline}
  n14 = DynamicParameter() {defaultValue=<rgba 0 1 1 1>, paramNames [Erode,XTile,PanAdd,SplineWidth]}
  n15 = AppendVector(A=@n14.XTile, B=NamedRerouteUsage())
  n16 = NamedRerouteDeclaration(None=Add(A=Multiply(A=@n11, B=@n12), B=Multiply(A=@n13, B=@n15))) {name=TexCoords, nodeColor=<rgba 1 0.0589 0.0398 1>}
  n17 = VectorParameter() {parameterName=MainTileAndPan, defaultValue=<rgba 0.3 1 -0.1 0.05>}
  n18 = AppendVector(A=@n17.R, B=@n17.R)
  n19 = VectorParameter() {parameterName=SecondaryTileAndPan, defaultValue=<rgba 0.6 1.4 -0.3 -0.05>}
  n20 = AppendVector(A=@n19.R, B=@n19.R)
  n21 = TextureSampleParameter2D(UVs=Panner(Coordinate=Multiply(A=@n16, B=@n20), Time=NamedRerouteUsage(), Speed=AppendVector(A=@n19.B, B=@n19.B))) {parameterName=SecondaryTexture, texture=/Game/Effects/Fort_Effects/Textures/Tile/T_Smooth_Tile_Chromatic}
  n22 = TextureCoordinate()
  n23 = StaticSwitchParameter(True=ComponentMask(None=@n22) {r=true}, False=ComponentMask(None=@n22) {g=true}) {parameterName=UseSpline}
  n24 = Multiply(A=Subtract(A=@n23) {constB=0.5}) {constB=5}
  n25 = Multiply(A=Multiply(A=@n21.G, B=Clamp(None=@n24) {maxDefault=0.5, minDefault=-0.5})) {constB=-0.4}
  n26 = Add(
   A=Panner(Coordinate=Multiply(A=@n16, B=@n18), Time=NamedRerouteUsage(), Speed=AppendVector(A=@n17.B, B=@n17.B))
   B=Multiply(A=@n25) {constB=0.25}
  )
  n27 = ScalarParameter() {defaultValue=0.5, parameterName=SecondaryTextureBrightness}
  n28 = Multiply(
   A=@n10
   B=Max(
    A=TextureSampleParameter2D(UVs=@n26) {parameterName=MainTexture, texture=/Game/Effects/Fort_Effects/Textures/Tile/T_ShieldCaustics_Tile_Chromatic}.RGB
    B=Multiply(A=@n21.RGB, B=@n27) {constB=0.5}
   )
  )
  n29 = Add(A=@n28) {constB=0.4}
  n30 = Add(A=@n28) {constB=-0.5}
  n31 = NamedRerouteUsage()
  n32 = Add(A=@n25, B=@n23)
  n33 = Constant() {r=0.5}
  n34 = ScalarParameter() {defaultValue=0.41, parameterName=EdgeFadeAdjust}
  n35 = Multiply(
   A=Multiply(A=@n10, B=SmoothStep(Min=@n29, Max=@n30, Value=@n31) {constMin=1})
   B=Power(
    Base=SphereMask(A=@n32, B=@n33, Radius=@n34) {attenuationRadius=0.4, hardnessPercent=0}
    Exp=ScalarParameter() {defaultValue=1.5, parameterName=SideOpacityPower}
   )
  )
  n36 = ScalarParameter() {defaultValue=1, parameterName=BaseOpacityMult}
  n37 = Multiply(A=@n6.A, B=Saturate(None=Multiply(A=@n35, B=@n36) {constB=2}))
  n38 = QualitySwitch(Default=DepthFade(Opacity=@n37, FadeDistance=ScalarParameter() {defaultValue=100, parameterName=DepthFade}), Low=@n37)
  n39 = ScalarParameter() {defaultValue=1, parameterName=AngleFadeSensitivity}
  LinearInterpolate(
   A=@n38
   B=Multiply(
    A=@n38
    B=MaterialFunctionCall("FadeSensitivity (optional)"=@n39) {materialFunction=&FadeBasedOnViewAngle}.Result
   )
   Alpha=ScalarParameter() {parameterName=UseAngleFade}
  )
  n40 = TextureCoordinate()
  n41 = LinearInterpolate(
   A=@n6.RGB
   B=LinearInterpolate(
    A=VectorParameter() {parameterName=Gradient_TopColor, defaultValue=<rgba 0.1563 0.0598 1 1>}.RGB
    B=VectorParameter() {parameterName=Gradient_BottomColor, defaultValue=<rgba 1 0.0412 0.5213 1>}.RGB
    Alpha=ComponentMask(None=@n40) {g=true}
   )
   Alpha=ScalarParameter() {parameterName=IsGradient}
  )
  n42 = Desaturation(None=Add(A=@n28) {constB=0.2}, Fraction=ScalarParameter() {defaultValue=-2, parameterName=Desaturation})
  n43 = DynamicParameter() {parameterIndex=1, defaultValue=<rgba 0 1 1 0>, paramNames [SplineWidthTip, YTileMult, PulseMaskTravel, PulseColorShift]}
  n44 = SphereMask(A=ComponentMask(None=TextureCoordinate() {uTiling=2}) {r=true}, B=LinearInterpolate(Alpha=@n43.PulseMaskTravel) {constA=-2, constB=4}) {attenuationRadius=2, hardnessPercent=0}
  n45 = NamedRerouteDeclaration(None=Multiply(A=Power(Base=@n44) {constExponent=3}, B=NamedRerouteDeclaration(None=@n44) {name=TravelMask, nodeColor=<rgba 0 1 0.1562 1>})) {name=PulseMask, nodeColor=<rgba 1 0.0469 0 1>}
  n46 = LinearInterpolate(
   A=VectorParameter() {parameterName=PulseColor2, defaultValue=<rgba 4 0.0833 0.625 1>}.RGB
   B=VectorParameter() {parameterName=PulseColor, defaultValue=<rgba 0 1.8049 3 1>}.RGB
   Alpha=@n43.PulseColorShift
  )
  MaterialFunctionCall(Emissive=Add(A=Multiply(A=@n45, B=@n46), B=Multiply(A=@n41, B=@n42))) {materialFunction=&MF_FX_EmissiveAdjustments}
  MaterialFunctionCall() {materialFunction=&DebugTimeSine}
  n47 = ScalarParameter() {defaultValue=-1, parameterName=VarianceMin}
  n48 = ScalarParameter() {defaultValue=0.5, parameterName=VsrianceMax}
  n49 = NamedRerouteUsage()
  n50 = VectorParameter() {parameterName=VarianceTileAndPan, defaultValue=<rgba 1.5 0.6 -0.07 0.1>}
  n51 = AppendVector(A=@n50.R, B=@n50.R)
  n52 = TextureSample(UVs=Panner(Coordinate=Multiply(A=@n49, B=@n51), Time=NamedRerouteUsage(), Speed=AppendVector(A=@n50.B, B=@n50.B)) {speedX=0.05, speedY=0.2}) {samplerType=SAMPLERTYPE_LinearGrayscale, texture=/Game/Effects/Fort_Effects/Textures/Tile/T_CellNoise_Tile_M}
  n53 = Max(A=OneMinus(None=NamedRerouteUsage()), B=@n14.Erode)
  n54 = ScalarParameter() {parameterName=TravelMaskDrivesErode}
  NamedRerouteDeclaration(None=StaticSwitchParameter(True=LinearInterpolate(A=@n47, B=@n48, Alpha=@n52.G) {constA=-1, constB=1.5}, False=LinearInterpolate(A=@n14.Erode, B=@n53, Alpha=@n54)) {parameterName=UseTextureVariance}) {name=Erode, nodeColor=<rgba 0.0547 0 1 1>}
  NamedRerouteDeclaration(None=@n14.SplineWidth) {name=SplineWidth, nodeColor=<rgba 0 1 0.6484 1>}
  NamedRerouteDeclaration(None=@n4.YTileMult) {name=YTileMult, nodeColor=<rgba 0 1 1 1>}
  NamedRerouteDeclaration(None=Add(A=@n14.PanAdd, B=Time())) {name=PanTime, nodeColor=<rgba 1 0.0175 0.0844 1>}
  Panner()
 ]
}
