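# M_Clean_Trail is a translucent, two-sided unlit master material for clean Niagara trails, also usable on sprites and meshes.
# It builds the streak procedurally from UV sphere masks, with dynamic parameters driving width, optional length reveal and edge-fade direction.
# Particle color or a two-color gradient, darkened by the mask, feeds emissive adjustments.
# Opacity combines particle alpha, non-mobile depth fade and optional edge masks, while SplineThicken optionally offsets vertices.
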Material /Game/Effects/Niagara/MasterMaterials/Ribbon/M_Clean_Trail
graph {
 properties {
  bUsedWithBeamTrails=true
  bUsedWithMeshDeformer=true
  bUsedWithNiagaraMeshParticles=true
  bUsedWithNiagaraRibbons=true
  bUsedWithNiagaraSprites=true
  bUsedWithSkeletalMesh=true
  blendMode=BLEND_Translucent
  pixelDepthOffsetMode=PDOM_Legacy
  shadingModel=MSM_Unlit
  twoSided=true
  usedShadingModels=Unlit
 }
 nodes [
  n1 = DynamicParameter() {parameterIndex=1, defaultValue=<rgba 1 1 0 1>, paramNames [SplineTip,SplineBase,FlipEdgeFade,Param4]}
  LinearInterpolate(
   B=MaterialFunctionCall(WidthBase=@n1.SplineBase, WidthTip=@n1.SplineBase, Expand U or V UV Channel=StaticBoolParameter() {parameterName=ExpandUorV}) {materialFunction=&SplineThicken}.WorldPositionOffset
   Alpha=ScalarParameter() {parameterName=UseSplineThicken}
  )
  MaterialFunctionCall() {materialFunction=&DebugTimeSine}
  n2 = TextureCoordinate()
  n3 = NamedRerouteDeclaration(None=LinearInterpolate(
    A=@n2
    B=MaterialFunctionCall(XY=@n2) {materialFunction=&Swizzle}.YX
    Alpha=ScalarParameter() {parameterName=SwizzleUV}
   )) {name=UV, nodeColor=<rgba 0.9219 0 1 1>}
  n4 = ParticleColor()
  n5 = NamedRerouteUsage()
  n6 = ComponentMask(None=@n5) {r=true}
  n7 = Constant() {r=0.5}
  n8 = ComponentMask(None=@n5) {g=true}
  n9 = DynamicParameter() {paramNames [Width, InnerTranslucency, LengthControl, Param4]}
  n10 = SphereMask(A=@n8, B=@n7, Radius=LinearInterpolate(Alpha=@n9.Width) {constA=0.01, constB=0.5}, Hardness=LinearInterpolate(Alpha=@n9.Width) {constB=0.9}) {hardnessPercent=70}
  n11 = SphereMask(A=@n8, B=@n7, Radius=@n7) {attenuationRadius=0.4, hardnessPercent=0}
  n12 = LinearInterpolate(
   A=LinearInterpolate(Alpha=Multiply(A=@n9.Width, B=@n9.Width)) {constA=1, constB=0.01}
   Alpha=Power(Base=OneMinus(None=@n11), Exp=ScalarParameter() {defaultValue=2.5, parameterName=Mid_Mask}) {constExponent=2.5}
  ) {constA=0.04}
  n13 = Add(A=@n12) {constB=0.1}
  n14 = Multiply(A=@n4.A, B=Multiply(A=SphereMask(A=@n6, B=@n7, Radius=@n7) {hardnessPercent=90}, B=Multiply(A=@n10, B=@n13)))
  n15 = LinearInterpolate(
   B=VectorParameter() {parameterName=DarkColor, defaultValue=<rgba 0.1718 0.2044 0.5 0>}.RGB
   Alpha=Multiply(A=ComponentMask(None=@n3) {r=true}, B=@n14)
  ) {constA=1}
  n16 = ComponentMask(None=TextureCoordinate()) {g=true}
  n17 = ScalarParameter() {defaultValue=2, parameterName=GradientAmount}
  n18 = LinearInterpolate(
   A=@n4.RGB
   B=LinearInterpolate(
    A=VectorParameter() {parameterName=Gradient_TopColor, defaultValue=<rgba 0 1 1 1>}.RGB
    B=VectorParameter() {parameterName=Gradient_BottomColor, defaultValue=<rgba 1 0.0005 0 1>}.RGB
    Alpha=Power(Base=@n16, Exp=@n17)
   )
   Alpha=ScalarParameter() {parameterName=IsGradient}
  )
  MaterialFunctionCall(Emissive=Multiply(A=Multiply(A=@n15, B=@n18), B=@n12)) {materialFunction=&MF_FX_EmissiveAdjustments}
  n19 = ScalarParameter() {defaultValue=30, parameterName=Depth_Fade}
  n20 = Multiply(
   A=@n14
   B=ShadingPathSwitch(Default=DepthFade(FadeDistance=@n19) {fadeDistanceDefault=30}, Mobile=Constant() {r=1}) {desc=Migrated from FeatureLevelSwitch}
  )
  n21 = NamedRerouteUsage()
  n22 = ComponentMask(None=@n21) {r=true}
  n23 = LinearInterpolate(A=@n22, B=OneMinus(None=@n22), Alpha=ScalarParameter() {parameterName=Mask inverted})
  n24 = ScalarParameter() {defaultValue=1.5, parameterName=Mask_power}
  n25 = Saturate(None=Multiply(A=Power(Base=@n23, Exp=@n24) {constExponent=1.5}, B=ScalarParameter() {defaultValue=5, parameterName=Mask_intensity}) {constB=5})
  n26 = ComponentMask(None=@n21) {r=true}
  n27 = Power(
   Base=LinearInterpolate(A=OneMinus(None=@n26), B=@n26, Alpha=ScalarParameter() {parameterName=Mask inverted})
   Exp=ScalarParameter() {defaultValue=1.5, parameterName=Mask_power_another_side}
  )
  n28 = ScalarParameter() {defaultValue=5, parameterName=Mask_intensity_another_side}
  n29 = Multiply(A=@n25, B=Saturate(None=Multiply(A=@n27, B=@n28)))
  n30 = ScalarParameter() {parameterName=Add_the_mask_on_another_side}
  n31 = LinearInterpolate(
   A=@n20
   B=Multiply(A=@n20, B=LinearInterpolate(A=@n25, B=@n29, Alpha=@n30))
   Alpha=ScalarParameter() {parameterName=Use the mask on the left edge}
  )
  n32 = NamedRerouteUsage()
  n33 = ComponentMask(None=@n32) {g=true}
  n34 = OneMinus(None=@n33)
  n35 = DynamicParameter() {parameterIndex=1, defaultValue=<rgba 1 1 0 1>, paramNames [SplineTip,SplineBase,FlipEdgeFade,Param4]}
  n36 = Power(Base=Multiply(A=LinearInterpolate(A=@n33, B=@n34, Alpha=@n35.FlipEdgeFade), B=ComponentMask(None=@n32) {r=true})) {constExponent=0.5}
  Multiply(
   A=LinearInterpolate(A=@n31, B=Multiply(A=@n36, B=@n31), Alpha=ScalarParameter() {parameterName=FadeEdge})
   B=StaticSwitchParameter(True=NamedRerouteUsage(), False=Constant() {r=1}) {parameterName=UseDynamicLengthControl}
  )
  n37 = OneMinus(None=DynamicParameter() {paramNames [Width, InnerTranslucency, LengthControl, Param4]}.LengthControl)
  n38 = ComponentMask(None=NamedRerouteUsage()) {r=true}
  NamedRerouteDeclaration(None=SmoothStep(Min=Saturate(None=@n37), Value=OneMinus(None=@n38))) {name=LengthControl, nodeColor=<rgba 1 0.9609 0 1>}
  TextureCoordinate()
  TextureCoordinate()
  TextureCoordinate()
 ]
}
