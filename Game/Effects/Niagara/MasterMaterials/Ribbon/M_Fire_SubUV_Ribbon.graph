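# Unlit translucent master material for animated flipbook fire on Niagara ribbons and sprites.
# It samples a packed 8x8 flipbook's RGB channels at offset phases, maxing them into temperature that drives curve-atlas color with unmultiply and hue shift.
# Opacity derives from temperature, with depth fade except at Low quality.
# UseAsSpriteMesh switches from dynamic-parameter timing, swizzled UVs and camera offset to vertex-color-and-time animation with billboarded vertex offset.
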Material /Game/Effects/Niagara/MasterMaterials/Ribbon/M_Fire_SubUV_Ribbon
graph {
 properties {
  bUsedWithNiagaraRibbons=true
  bUsedWithNiagaraSprites=true
  blendMode=BLEND_Translucent
  pixelDepthOffsetMode=PDOM_Legacy
  shadingModel=MSM_Unlit
  usedShadingModels=Unlit
 }
 nodes [
  Constant() {r=1}
  Constant()
  n1 = ScalarParameter() {defaultValue=1, parameterName=EmissiveMult}
  n2 = ParticleColor()
  n3 = VertexColor()
  n4 = Time()
  n5 = ScalarParameter() {defaultValue=1, parameterName=AnimSpeed}
  n6 = DynamicParameter() {desc=G - packed channel switch/weight, defaultValue=<rgba 0.5 0 0 0>, paramNames [AnimTime,Param2,Param3,Param4]}
  n7 = StaticSwitchParameter(True=Multiply(A=Add(A=@n3.R, B=@n4), B=Multiply(A=@n3.G, B=@n5)), False=@n6.AnimTime) {parameterName=UseAsSpriteMesh}
  n8 = Multiply(A=@n7) {constB=1.1}
  n9 = Constant() {r=8}
  n10 = TextureCoordinate()
  n11 = MaterialFunctionCall(XY=@n10) {materialFunction=&Swizzle}
  n12 = StaticSwitchParameter(True=@n10, False=StaticSwitchParameter(True=OneMinus(None=@n11.YX), False=@n11.YX) {parameterName=FlipTexture}) {parameterName=UseAsSpriteMesh}
  n13 = Multiply(A=@n7) {constB=0.9}
  n14 = Max(
   A=TextureSampleParameter2D(UVs=MaterialFunctionCall("Animation  Phase (0-1)"=@n8, Number of Rows=@n9, Number of Columns=@n9, UVs=@n12) {materialFunction=&FlipBook}.UVs) {parameterName=FireMask, texture=/Game/Effects/Fort_Effects/Textures/Fire/2DGasTorch_Flipbk_packed}.R
   B=TextureSampleParameter2D(UVs=MaterialFunctionCall("Animation  Phase (0-1)"=@n13, Number of Rows=@n9, Number of Columns=@n9, UVs=@n12) {materialFunction=&FlipBook}.UVs) {parameterName=FireMask, texture=&2DGasTorch_Flipbk_packed}.G
  )
  n15 = TextureSampleParameter2D(UVs=MaterialFunctionCall("Animation  Phase (0-1)"=@n7, Number of Rows=@n9, Number of Columns=@n9, UVs=@n12) {materialFunction=&FlipBook}.UVs) {parameterName=FireMask, texture=&2DGasTorch_Flipbk_packed}
  n16 = Multiply(
   A=Power(
    Base=Max(A=@n14, B=@n15.B)
    Exp=ScalarParameter() {defaultValue=1.25, group=Flame Channels, parameterName=Temperature Exponent, sliderMax=3, sortPriority=5}
   )
   B=ScalarParameter() {defaultValue=5, group=Flame Channels, parameterName=Temperature Scale, sliderMax=25, sortPriority=5}
  )
  n17 = Multiply(
   A=Power(Base=@n16, Exp=ScalarParameter() {defaultValue=1.5, group=Output Color, parameterName=Opacity Exponent, sliderMax=3, sortPriority=8})
   B=ScalarParameter() {defaultValue=3, group=Output Color, parameterName=Opacity Scale, sliderMax=3, sortPriority=8}
  ) {constB=3}
  n18 = CurveAtlasRowParameter(CurveTime=@n16) {
   atlas=/Game/Curie/Effects/temp/Curves/CA_NonCurie_FireVisuals
   curve=/Game/Curie/Effects/temp/Curves/Curve_Fiery_B
   parameterName=NonCurieCurves
  }
  n19 = Divide(A=@n18, B=Saturate(None=@n17))
  n20 = ScalarParameter() {defaultValue=0.45, group=Output Color, parameterName=UnMult, sliderMax=1}
  n21 = Multiply(
   A=@n2.RGB
   B=Multiply(
    A=LinearInterpolate(A=@n18, B=@n19, Alpha=@n20)
    B=ScalarParameter() {defaultValue=2, group=Output Color, parameterName=Final Color Brightness, sliderMax=50, sortPriority=9}
   )
  )
  MaterialFunctionCall(Emissive=MaterialFunctionCall(Hue Shift Percentage=ScalarParameter() {parameterName=HueShiftValue}, Texture=Multiply(A=@n1, B=@n21)) {materialFunction=&HueShift}.Result) {materialFunction=&MF_FX_EmissiveAdjustments}
  n22 = Multiply(A=@n2.A, B=@n17)
  n23 = DepthFade(Opacity=@n22, FadeDistance=ScalarParameter() {defaultValue=15, parameterName=FadeDistance, sliderMax=300})
  QualitySwitch(Default=@n23, Low=@n22, Epic=@n23) {desc=quality Opactiy}
  n24 = ScalarParameter() {defaultValue=15, parameterName=CameraOffset}
  n25 = VectorParameter() {parameterName=SpriteSizeAndOffset, defaultValue=<rgba 50 100 30 0>}
  n26 = Add(
   A=MaterialFunctionCall() {materialFunction=&ObjectPivotPoint}.Object Pivot Location
   B=Multiply(A=@n25.B, B=Constant3Vector() {constant=<rgba 0 0 1 0>})
  )
  n27 = AppendVector(A=@n25.R, B=@n25.R)
  StaticSwitchParameter(
   True=Add(
    A=MaterialFunctionCall(Offset Amount=@n24) {materialFunction=/Engine/Functions/Engine_MaterialFunctions02/WorldPositionOffset/CameraOffset}.Camera Offset
    B=MaterialFunctionCall!(Center Location=@n26, 0-1 UVs, XY scale=@n27, OverridePivot, "Normals (Optional)") {materialFunction=&Sprite}.World position offset
   )
   False=MaterialFunctionCall(Offset Amount=ScalarParameter() {parameterName=RibbonCameraOffset}) {materialFunction=&CameraOffset}.Camera Offset
  ) {parameterName=UseAsSpriteMesh}
 ]
}
