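# Unlit translucent Niagara ribbon material that draws a music-staff trail.
# Powered sine waves on UV1 form repeated horizontal lines and panning bar lines confined to the middle band, optionally thresholded into hard edges.
# A normal-map-distorted panning noise modulates emissive brightness, tinted by a curve-atlas gradient with dynamic-parameter desaturation.
# Opacity combines the staff mask with gradient alpha along U, and particle color scales both color and opacity.
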Material /Game/Effects/Niagara/MasterMaterials/Ribbon/M_MusicScore_Trail
graph {
 properties {bUsedWithNiagaraRibbons=true, blendMode=BLEND_Translucent, pixelDepthOffsetMode=PDOM_Legacy, shadingModel=MSM_Unlit, usedShadingModels=Unlit}
 nodes [
  n1 = TextureCoordinate()
  n2 = VectorParameter() {group=Distortion, parameterName=Distortion UV, sortPriority=1, defaultValue=<rgba 1 1 0 0>}
  n3 = AppendVector(A=@n2.R, B=@n2.R)
  n4 = TextureSampleParameter2D(UVs=Panner(Coordinate=Multiply(A=@n1, B=@n3), Speed=AppendVector(A=@n2.B, B=@n2.B))) {
   group=Distortion
   parameterName="Distortion  Tex"
   samplerType=SAMPLERTYPE_Normal
   sortPriority=0
   texture=&T_Weap_Distort_N_02
  }
  n5 = ComponentMask(None=@n4.RGB) {g=true, r=true}
  n6 = ScalarParameter() {defaultValue=0.2, group=Distortion, parameterName=Distortion Intensity, sortPriority=2}
  n7 = VectorParameter() {group=Noise, parameterName=Noise UV, sortPriority=6, defaultValue=<rgba 2 0.5 -4 4>}
  n8 = Multiply(A=TextureCoordinate(), B=AppendVector(A=@n7.R, B=@n7.R))
  n9 = AppendVector(A=@n7.B, B=@n7.B)
  n10 = TextureSampleParameter2D(UVs=Add(A=Multiply(A=@n5, B=@n6), B=Panner(Coordinate=@n8, Speed=@n9))) {group=Noise, parameterName=Noise Tex, samplerType=SAMPLERTYPE_LinearColor, texture=/Game/Global/Textures/Default/Noise/T_Noise01}
  n11 = CurveAtlasRowParameter(CurveTime=ComponentMask(None=TextureCoordinate()) {g=true}) {
   atlas=/Game/Athena/LinearColorCurves/A_Atlas_Turtleneck_Glider
   curve=/Game/Weapons/FORT_Melee/Pickaxe_Calavera/FX/Curve_Pickaxe_Calavera_Color
   defaultValue=41
  }
  n12 = Multiply(
   A=LinearInterpolate(
    A=ScalarParameter() {defaultValue=0.1, group=Emissive, parameterName=Emissive Dark}
    B=ScalarParameter() {defaultValue=3, group=Emissive, parameterName=Emissive Light}
    Alpha=@n10.R
   ) {constA=0.1, constB=1.5}
   B=Desaturation(None=@n11, Fraction=DynamicParameter() {defaultValue=<rgba 0 1 1 1>, paramNames [Desaturation,Param2,Param3,Param4]}.Desaturation)
  )
  n13 = ScalarParameter() {defaultValue=0.5, group=Emissive, parameterName=Emissive Intensity}
  n14 = ParticleColor()
  MaterialFunctionCall(Emissive=Multiply(A=Multiply(A=@n12, B=@n13), B=@n14.RGB)) {materialFunction=&MF_FX_EmissiveAdjustments}
  n15 = ComponentMask(None=TextureCoordinate() {coordinateIndex=1, vTiling=10}) {g=true}
  n16 = Sine(None=Add(A=Multiply(A=@n15) {constB=0.5}) {constB=-0.25})
  n17 = ScalarParameter() {defaultValue=6, group=Staff, parameterName=Horizon Power}
  n18 = TextureCoordinate() {coordinateIndex=1, uTiling=2}
  n19 = Add(A=Multiply(A=ComponentMask(None=@n18) {r=true}) {constB=0.5}) {constB=0.22}
  n20 = AppendVector(
   A=ScalarParameter() {defaultValue=2, group=Staff, parameterName=Vertical Pan Y}
   B=ScalarParameter() {group=Staff, parameterName=Vertical Pan X}
  )
  n21 = Power(
   Base=Sine(None=Panner(Coordinate=@n19, Speed=@n20) {speedX=2})
   Exp=ScalarParameter() {defaultValue=200, group=Staff, parameterName=Vertical Power}
  ) {constExponent=400}
  n22 = Constant() {r=0.09}
  n23 = Constant() {r=1}
  n24 = Constant()
  n25 = TextureCoordinate() {coordinateIndex=1}
  n26 = Multiply(A=ComponentMask(None=@n25) {g=true}, B=OneMinus(None=ComponentMask(None=@n25) {g=true}))
  n27 = Add(
   A=Saturate(None=Power(Base=@n16, Exp=@n17) {constExponent=6})
   B=Multiply(A=Saturate(None=@n21), B=If(A=@n26, B=@n22, "A > B"=@n23, "A == B"=@n23, "A < B"=@n24))
  )
  n28 = Constant() {r=1}
  n29 = StaticSwitchParameter(True=If(A=ComponentMask(None=@n27) {g=true}, B=Constant() {r=0.5}, "A > B"=@n28, "A == B"=@n28, "A < B"=Constant()), False=@n27) {defaultValue=true, group=Staff, parameterName=Use Hard Line, sortPriority=0}
  n30 = CurveAtlasRowParameter(CurveTime=ComponentMask(None=TextureCoordinate()) {r=true}) {
   atlas=&A_Atlas_Turtleneck_Glider
   curve=&Curve_Pickaxe_Calavera_Color
   defaultValue=41
  }
  Multiply(A=@n14.A, B=Multiply(A=Multiply(A=@n29, B=@n30.A), B=ScalarParameter() {defaultValue=1, group=Opactiy, parameterName=Opactiy Intensity}))
  ScalarParameter() {group=Emissive, parameterName=Desaturation}
 ]
}
