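# Translucent, unlit, two-sided master material for pickaxe slash trails on Niagara ribbons and mesh particles, with noise-based dissolve.
# A channel-selectable shape texture drives opacity with particle alpha, and emissive with particle color.
# HeightLerp erodes opacity against a noise texture using the dynamic Dissolve parameter.
# View-angle fade, blendable lead and tail fades along U, and depth fade further soften the trail.
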Material /Game/Effects/Niagara/MasterMaterials/Ribbon/M_Pickaxe_SlashDissolve_Trail
graph {
 properties {
  bAutomaticallySetUsageInEditor=false
  bUsedWithNiagaraMeshParticles=true
  bUsedWithNiagaraRibbons=true
  blendMode=BLEND_Translucent
  shadingModel=MSM_Unlit
  twoSided=true
  usedShadingModels=Unlit
 }
 nodes [
  n1 = Power(Base=ComponentMask(None=TextureCoordinate()) {r=true}, Exp=ScalarParameter() {defaultValue=1, parameterName=FadePower})
  n2 = ScalarParameter() {defaultValue=3, parameterName=FadeIntensity}
  n3 = LinearInterpolate(B=Saturate(None=Multiply(A=@n1, B=@n2) {constB=3}), Alpha=ScalarParameter() {parameterName=UseLeadFade}) {constA=1}
  n4 = ParticleColor()
  n5 = TextureCoordinate()
  n6 = ComponentMask(None=VectorParameter() {parameterName=AlphaUVScale, defaultValue=<rgba 1 1.1 0 0>}.RGB) {g=true, r=true}
  n7 = VectorParameter() {parameterName=UVOffset, defaultValue=<rgba 0 -0.05 0 0>}
  n8 = TextureSampleParameter2D(UVs=Add(A=Multiply(A=@n5, B=@n6), B=ComponentMask(None=@n7.RGB) {g=true, r=true})) {
   parameterName=AlphaShape
   samplerSource=SSM_Clamp_WorldGroupSettings
   texture=/Game/Weapons/FORT_Melee/Pickaxe_Future_Samurai_Summer_Male/FX/T_Pickaxe_Future_Samurai_Summer_Male_Trail_01
  }
  n9 = StaticComponentMaskParameter(None=@n8.RGB) {defaultR=true, parameterName=RGBMask}
  n10 = Multiply(A=Multiply(A=@n4.A, B=@n9), B=ScalarParameter() {defaultValue=5, parameterName=AlphaShapeContrast})
  n11 = LinearInterpolate(Alpha=DynamicParameter() {defaultValue=<rgba 1 1 0 1>, paramNames [Param1,Param2,Dissolve,Param4]}.Dissolve)
  n12 = TextureCoordinate()
  n13 = ComponentMask(None=VectorParameter() {parameterName=DissUV, defaultValue=<rgba 3 0.5 0 0>}.RGB) {g=true, r=true}
  n14 = VectorParameter() {parameterName=Dissoffset, defaultValue=<rgba 0.5 0 0 0>}
  n15 = TextureSampleParameter2D(UVs=Add(A=Multiply(A=@n12, B=@n13), B=ComponentMask(None=@n14.RGB) {g=true, r=true})) {parameterName=DissolveNoise, texture=/Game/Effects/Niagara/MasterMaterials/SplineThickenTrail/Texture/T_Pickaxe_SoapPocket_NoiseGlow}
  n16 = ScalarParameter() {defaultValue=2, parameterName="Contrast (S)"}
  n17 = Multiply(
   A=Multiply(
    A=Saturate(None=@n10)
    B=MaterialFunctionCall(Transition Phase=@n11, Height Texture=@n15.R, Contrast=@n16) {materialFunction=&HeightLerp}.Results
   )
   B=MaterialFunctionCall("FadeSensitivity (optional)"=ScalarParameter() {defaultValue=1, parameterName="FadeSensitivity (optional) (S)"}) {materialFunction=&FadeBasedOnViewAngle}.Result
  )
  n18 = OneMinus(None=ComponentMask(None=TextureCoordinate()) {r=true})
  n19 = ScalarParameter() {defaultValue=1, parameterName=FadePower}
  n20 = Saturate(None=Multiply(A=Power(Base=@n18, Exp=@n19), B=ScalarParameter() {defaultValue=3, parameterName=FadeIntensity}) {constB=3})
  n21 = ScalarParameter() {parameterName=UseTailFade}
  DepthFade(
   Opacity=Multiply(A=Multiply(A=@n17, B=@n3), B=LinearInterpolate(B=@n20, Alpha=@n21) {constA=1})
   FadeDistance=ScalarParameter() {defaultValue=100, parameterName=FadeDistance}
  )
  n22 = Power(Base=@n9, Exp=ScalarParameter() {defaultValue=1, parameterName=EmissivePower})
  MaterialFunctionCall(Emissive=Multiply(A=Multiply(A=@n4.RGB, B=@n22))) {materialFunction=&MF_FX_EmissiveAdjustments}
  MaterialFunctionCall() {materialFunction=&DebugTimeSine}
 ]
}
