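# Translucent, unlit, two-sided master material for rainbow trails, usable on Niagara ribbons and mesh particles.
# Three panning caustic samples, offset along U and warped by smoke, fill separate RGB channels for a color-split pattern modulated by another scrolling texture.
# That removable pattern is added to a swizzled rainbow gradient, then tinted by particle color for emissive.
# Opacity combines particle alpha, view-angle fade, an edge mask across V, and adjustable lead and tail masks.
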Material /Game/Effects/Niagara/MasterMaterials/Ribbon/M_Rainbow_Trail
graph {
 properties {
  bUsedWithMeshDeformer=true
  bUsedWithNiagaraMeshParticles=true
  bUsedWithNiagaraRibbons=true
  bUsedWithSkeletalMesh=true
  blendMode=BLEND_Translucent
  pixelDepthOffsetMode=PDOM_Legacy
  shadingModel=MSM_Unlit
  twoSided=true
  usedShadingModels=Unlit
 }
 nodes [
  n1 = TextureCoordinate() {uTiling=0.3}
  n2 = Constant() {r=0.03}
  n3 = Panner(Coordinate=Add(A=@n1, B=@n2) {constB=0.05}) {speedX=-0.4}
  n4 = TextureCoordinate() {uTiling=0.7}
  n5 = Multiply(A=TextureSample(UVs=Panner(Coordinate=@n4) {speedX=-0.1}) {texture=&T_Inky_Smoke_Tile}.R) {constB=0.1}
  n6 = Panner(Coordinate=@n1) {speedX=-0.4}
  n7 = Panner(Coordinate=Add(A=@n1, B=Multiply(A=@n2) {constB=-1}) {constB=-0.05}) {speedX=-0.4}
  n8 = MaterialFunctionCall(
   X=TextureSample(UVs=Add(A=@n3, B=@n5)) {samplerType=SAMPLERTYPE_LinearGrayscale, texture=/Game/Global/Textures/Default/Caustics/T_Nosie_Caustics_Grayscale}.R
   Y=TextureSample(UVs=Add(A=@n6, B=@n5)) {samplerType=SAMPLERTYPE_LinearGrayscale, texture=&T_Nosie_Caustics_Grayscale}.R
   Z=TextureSample(UVs=Add(A=@n7, B=@n5)) {samplerType=SAMPLERTYPE_LinearGrayscale, texture=&T_Nosie_Caustics_Grayscale}.R
  ) {materialFunction=&MakeFloat3}
  n9 = Power(Base=Multiply(A=Power(Base=@n8.Result) {constExponent=1}) {constB=2}) {constExponent=1.2}
  n10 = Round(None=@n9)
  n11 = TextureSample(UVs=Panner(Coordinate=TextureCoordinate() {uTiling=0.8, vTiling=2}) {speedX=-0.1, speedY=0.3}) {texture=/Game/Effects/Fort_Effects/Textures/Power_Source/T_Power_Source_Sphere_Texture}
  n12 = Multiply(A=Multiply(A=LinearInterpolate(A=@n10, B=@n9) {constAlpha=0.9}) {constB=1.5}, B=Saturate(None=LinearInterpolate(Alpha=@n11.R) {constA=-0.2}))
  n13 = LinearInterpolate(
   A=Desaturation(None=Multiply(A=@n12) {constB=1.3}, Fraction=Constant() {r=-0.5})
   B=Constant()
   Alpha=ScalarParameter() {group=Colorful_Texture, parameterName=Remove_Colorful_texure}
  )
  n14 = MaterialFunctionCall(XY=TextureCoordinate()) {materialFunction=&Swizzle}
  n15 = ScalarParameter() {defaultValue=1, group=Emissive, parameterName=UV_Tiling}
  n16 = VectorParameter() {group=Emissive, parameterName=UV_offset, defaultValue=<rgba 0 0 0 1>}
  n17 = TextureSampleParameter2D(UVs=Add(A=Multiply(A=@n14.YX, B=@n15) {constB=2}, B=ComponentMask(None=@n16.RGB) {g=true, r=true})) {
   group=Emissive
   parameterName=Default_Tex
   texture=/Game/Gadgets/Assets/VinderTech_GliderChute/Glider_Peach_Male/FX/T_Glider_Peach_rainbow_gradient
  }
  n18 = Desaturation(None=@n17.RGB, Fraction=ScalarParameter() {defaultValue=-0.3, group=Emissive, parameterName=Desaturation})
  n19 = LinearInterpolate(A=Round(None=@n18), B=@n18, Alpha=Constant() {r=0.95}) {constAlpha=0.7}
  MaterialFunctionCall(Emissive=Multiply(A=ParticleColor().RGB, B=Add(A=@n13, B=@n19))) {materialFunction=&MF_FX_EmissiveAdjustments}
  n20 = ParticleColor()
  n21 = MaterialFunctionCall("FadeSensitivity (optional)"=ScalarParameter() {defaultValue=0.5, parameterName=FadeBasedOnViewAngle}) {materialFunction=&FadeBasedOnViewAngle}
  n22 = ComponentMask(None=TextureCoordinate()) {g=true}
  n23 = Multiply(A=@n22, B=OneMinus(None=@n22))
  n24 = Multiply(
   A=Saturate(None=Multiply(A=@n20.A, B=@n21.Result))
   B=MaterialFunctionCall(In=Clamp(None=@n23), Contrast=ScalarParameter() {parameterName=MaskContrast}) {materialFunction=&CheapContrast}.Result
  )
  n25 = TextureCoordinate()
  n26 = Multiply(A=OneMinus(None=ComponentMask(None=@n25) {r=true}), B=ScalarParameter() {defaultValue=2.5, group=Alpha, parameterName=Tail_Edge_hardness}) {constB=2.5}
  n27 = ScalarParameter() {defaultValue=2, group=Alpha, parameterName=Tail_Mask_Power}
  n28 = Multiply(A=ComponentMask(None=TextureCoordinate()) {r=true}, B=ScalarParameter() {defaultValue=2, group=Alpha, parameterName=Lead_Edge_hardness}) {constB=2}
  n29 = ScalarParameter() {defaultValue=2, group=Alpha, parameterName=Lead_Mask_Power}
  n30 = Multiply(A=Saturate(None=Power(Base=@n26, Exp=@n27) {constExponent=4}), B=Saturate(None=Power(Base=@n28, Exp=@n29)))
  Multiply(A=Multiply(A=Multiply(A=@n24, B=@n30), B=ScalarParameter() {defaultValue=1, parameterName=Alpha_Main}))
 ]
}
