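# Master translucent unlit material for Niagara ribbon and mesh-particle effects, with most features blended in by scalar "Use" parameters.
# Color lerps ColorA/B by mask, gradient or texture, tinted by particle color and optional time-of-day.
# Opacity combines a distortable mask, edge masks, gradient-based erosion and depth, camera and angle fades.
# Dynamic parameters drive scrolling, erosion and switch-gated WPO bend/SplineThicken; scroll and target vectors read only their X components.
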Material /Game/Effects/Niagara/MasterMaterials/Ribbon/M_RibbonMesh_Master
graph {
 properties {
  bAutomaticallySetUsageInEditor=false
  bUsedWithNiagaraMeshParticles=true
  bUsedWithNiagaraRibbons=true
  blendMode=BLEND_Translucent
  shadingModel=MSM_Unlit
  usedShadingModels=Unlit
 }
 nodes [
  n1 = VectorParameter() {group=Color, parameterName=Color Morning Multiplier, sortPriority=11, defaultValue=<rgba 0.9896 0.268 0.4929 0>}
  n2 = VectorParameter() {group=Color, parameterName=Color Evening Multiplier, sortPriority=13, defaultValue=<rgba 0.599 0.2901 0.3358 0>}
  n3 = VectorParameter() {group=Color, parameterName=Color Night Multiplier, sortPriority=14, defaultValue=<rgba 0.3498 0.4864 0.6458 0>}
  n4 = MaterialFunctionCall(
   Color_Morning=@n1.RGB
   Color_Day=VectorParameter() {group=Color, parameterName=Color Day Multiplier, sortPriority=12, defaultValue=<rgba 1 1 1 0>}.RGB
   Color_Evening=@n2.RGB
   Color_Night=@n3.RGB
  ) {materialFunction=&Color_Time_of_Day}
  n5 = ScalarParameter() {group=Edge Masks, parameterName=Back Edge Mask UV Channel, sortPriority=7}
  n6 = TextureCoordinate()
  n7 = TextureCoordinate() {coordinateIndex=1}
  n8 = OneMinus(None=ChannelMaskParameter(None=LinearInterpolate(A=@n6, B=@n7, Alpha=@n5)) {group=Edge Masks, parameterName=Back Edge Mask Channel, sortPriority=6}.R)
  n9 = LinearInterpolate(
   A=TextureCoordinate()
   B=TextureCoordinate() {coordinateIndex=1}
   Alpha=ScalarParameter() {group=Opacity Mask, parameterName=Mask UV Channel, sortPriority=2}
  )
  n10 = MaterialFunctionCall(XY=@n9) {materialFunction=&Swizzle}
  n11 = ScalarParameter() {group=Opacity Mask, parameterName=Swizzle Mask UVs, sortPriority=3}
  n12 = ScalarParameter() {defaultValue=1, group=Opacity Mask, parameterName=Mask UV ScaleX, sortPriority=4}
  n13 = ScalarParameter() {defaultValue=1, group=Opacity Mask, parameterName=Mask UV ScaleY, sortPriority=5}
  n14 = ScalarParameter() {defaultValue=-1, group=Distortion, parameterName=Distortion LerpA, sortPriority=10}
  n15 = ScalarParameter() {defaultValue=1, group=Distortion, parameterName=Distortion LerpB, sortPriority=11}
  n16 = LinearInterpolate(
   A=TextureCoordinate()
   B=TextureCoordinate() {coordinateIndex=1}
   Alpha=ScalarParameter() {group=Distortion, parameterName=Distortion UV Channel, sortPriority=2}
  )
  n17 = Multiply(
   A=LinearInterpolate(
    A=@n16
    B=MaterialFunctionCall(XY=@n16) {materialFunction=&Swizzle}.YX
    Alpha=ScalarParameter() {group=Distortion, parameterName=Swizzle Distortion UVs, sortPriority=3}
   )
   B=AppendVector(
    A=ScalarParameter() {defaultValue=1, group=Distortion, parameterName=Distortion UV ScaleX, sortPriority=4}
    B=ScalarParameter() {defaultValue=1, group=Distortion, parameterName=Distortion UV ScaleY, sortPriority=5}
   )
  )
  n18 = DynamicParameter() {defaultValue=<rgba 0 0 0 0>, paramNames [Mask ScrollX, Mask ScrollY, Distortion ScrollX, Distortion ScrollY]}
  n19 = AppendVector(A=@n18.Distortion ScrollX, B=@n18.Distortion ScrollX)
  n20 = ScalarParameter() {group=Distortion, parameterName=Distortion PanX, sortPriority=6}
  n21 = ScalarParameter() {group=Distortion, parameterName=Distortion PanY, sortPriority=7}
  n22 = TextureSampleParameter2D(UVs=Panner(Coordinate=Add(A=@n17, B=@n19), Speed=AppendVector(A=@n20, B=@n21))) {
   group=Distortion
   parameterName=Distortion Texture
   sortPriority=0
   texture=/Game/Effects/Fort_Effects/Textures/Power_Source/T_Power_Source_Sphere_Texture
  }
  n23 = Power(
   Base=Multiply(
    A=ChannelMaskParameter(None=@n22.RGB) {group=Distortion, parameterName=Distortion Texture Channel, sortPriority=1}.R
    B=ScalarParameter() {defaultValue=1, group=Distortion, parameterName=Distortion Multiply, sortPriority=8}
   )
   Exp=ScalarParameter() {defaultValue=1, group=Distortion, parameterName=Distortion Power, sortPriority=9}
  )
  n24 = Add(
   A=Multiply(A=LinearInterpolate(A=@n9, B=@n10.YX, Alpha=@n11), B=AppendVector(A=@n12, B=@n13))
   B=LinearInterpolate(B=LinearInterpolate(A=@n14, B=@n15, Alpha=@n23), Alpha=ScalarParameter() {group=Distortion, parameterName=UseDistortion, sortPriority=-1})
  )
  n25 = DynamicParameter() {defaultValue=<rgba 0 0 0 0>, paramNames [Mask ScrollX, Mask ScrollY, Distortion ScrollX, Distortion ScrollY]}
  n26 = AppendVector(A=@n25.Mask ScrollX, B=@n25.Mask ScrollX)
  n27 = ScalarParameter() {group=Opacity Mask, parameterName=Mask PanX, sortPriority=6}
  n28 = ScalarParameter() {group=Opacity Mask, parameterName=Mask PanY, sortPriority=7}
  n29 = TextureSampleParameter2D(UVs=Panner(Coordinate=Add(A=@n24, B=@n26), Speed=AppendVector(A=@n27, B=@n28))) {
   group=Opacity Mask
   parameterName=Mask Texture
   sortPriority=0
   texture=/Game/Effects/Fort_Effects/Textures/Smoke/T_Smoke_Trail_Soft_Chromatic
  }
  n30 = ChannelMaskParameter(None=@n29.RGB) {group=Opacity Mask, parameterName=Mask Texture Channel, sortPriority=1}
  n31 = LinearInterpolate(A=@n30.R, B=@n8, Alpha=ScalarParameter() {group=Color, parameterName=UseLerpColorGradient, sortPriority=30})
  n32 = ScalarParameter() {defaultValue=1, group=Color, parameterName=Color Alpha Multiply, sortPriority=31}
  n33 = Saturate(None=Power(Base=Multiply(A=@n31, B=@n32), Exp=ScalarParameter() {defaultValue=1, group=Color, parameterName=Color Alpha Power}) {constExponent=1})
  n34 = LinearInterpolate(
   A=TextureCoordinate()
   B=TextureCoordinate() {coordinateIndex=1}
   Alpha=ScalarParameter() {group=Color, parameterName=Color UV Channel, sortPriority=22}
  )
  n35 = LinearInterpolate(
   A=@n34
   B=MaterialFunctionCall(XY=@n34) {materialFunction=&Swizzle}.YX
   Alpha=ScalarParameter() {group=Color, parameterName=Swizzle Color UVs, sortPriority=23}
  )
  n36 = AppendVector(
   A=ScalarParameter() {defaultValue=1, group=Color, parameterName=Color UV ScaleX, sortPriority=24}
   B=ScalarParameter() {defaultValue=1, group=Color, parameterName=Color UV ScaleY, sortPriority=25}
  )
  n37 = ScalarParameter() {group=Color, parameterName=Color PanX, sortPriority=26}
  n38 = ScalarParameter() {group=Color, parameterName=Color PanY, sortPriority=27}
  n39 = TextureSampleParameter2D(UVs=Panner(Coordinate=Multiply(A=@n35, B=@n36), Speed=AppendVector(A=@n37, B=@n38))) {
   group=Color
   parameterName=Color Texture
   sortPriority=21
   texture=/Game/Effects/Niagara/MasterMaterials/SplineThickenTrail/Texture/T_Pickaxe_SoapPocket_NoiseGlow
  }
  n40 = StaticComponentMaskParameter(None=@n39.RGB) {defaultR=true, group=Color, parameterName=Color Texture Channel, sortPriority=21}
  n41 = ScalarParameter() {defaultValue=1, group=Color, parameterName=Color Texture Multiply, sortPriority=28}
  n42 = Multiply(
   A=Power(Base=Multiply(A=@n40, B=@n41), Exp=ScalarParameter() {defaultValue=1, group=Color, parameterName=Color Texture Power, sortPriority=29})
   B=@n33
  )
  n43 = ScalarParameter() {group=Color, parameterName=UseColorTexture, sortPriority=20}
  n44 = ParticleColor()
  n45 = Multiply(
   A=LinearInterpolate(
    A=VectorParameter() {group=Color, parameterName=ColorA, sortPriority=0, defaultValue=<rgba 0 0 0 1>}.RGB
    B=VectorParameter() {group=Color, parameterName=ColorB, sortPriority=1, defaultValue=<rgba 1 1 1 1>}.RGB
    Alpha=LinearInterpolate(A=@n33, B=@n42, Alpha=@n43)
   )
   B=@n44.RGB
  )
  MaterialFunctionCall(Emissive=LinearInterpolate(
    A=Multiply(A=@n4.Result, B=@n45)
    B=@n45
    Alpha=ScalarParameter() {defaultValue=1, group=Color, parameterName=UseTimeofDayColor, sortPriority=10}
   ) {constAlpha=1}) {materialFunction=&MF_FX_EmissiveAdjustments}
  n46 = ScalarParameter() {defaultValue=1, group=Opacity Mask, parameterName=Mask Multiply, sortPriority=8}
  n47 = Saturate(None=Power(Base=Multiply(A=@n30.R, B=@n46), Exp=ScalarParameter() {defaultValue=1, group=Opacity Mask, parameterName=Mask Power, sortPriority=9}))
  n48 = ScalarParameter() {defaultValue=1, group=Edge Masks, parameterName=Back Edge Mask Multiply, sortPriority=8}
  n49 = Saturate(None=Power(Base=Multiply(A=@n8, B=@n48), Exp=ScalarParameter() {defaultValue=1, group=Edge Masks, parameterName=Back Edge Mask Power, sortPriority=9}))
  n50 = ScalarParameter() {group=Edge Masks, parameterName=Front Edge Mask UV Channel, sortPriority=3}
  n51 = Multiply(
   A=ChannelMaskParameter(None=LinearInterpolate(A=@n6, B=@n7, Alpha=@n50)) {group=Edge Masks, parameterName=Front Edge Mask Channel, sortPriority=2}.R
   B=ScalarParameter() {defaultValue=5, group=Edge Masks, parameterName=Front Edge Mask Multiply, sortPriority=4}
  )
  n52 = ScalarParameter() {defaultValue=2, group=Edge Masks, parameterName=Front Edge Mask Power, sortPriority=5}
  n53 = LinearInterpolate(
   A=Constant() {r=1}
   B=Saturate(None=Power(Base=@n51, Exp=@n52))
   Alpha=ScalarParameter() {defaultValue=1, group=Edge Masks, parameterName=UseFrontEdgeMask, sortPriority=0}
  )
  n54 = LinearInterpolate(
   A=@n53
   B=Multiply(A=@n49, B=@n53)
   Alpha=ScalarParameter() {defaultValue=1, group=Edge Masks, parameterName=UseBackEdgeMask, sortPriority=1}
  )
  n55 = Constant() {r=1}
  n56 = Multiply(A=@n49, B=ScalarParameter() {defaultValue=1, group=Erosion, parameterName=Erosion Gradient Multiply, sortPriority=11})
  n57 = DynamicParameter() {parameterIndex=1, defaultValue=<rgba 0 0 0.5 0>, paramNames [Erosion ScrollX, Erosion ScrollY, Erosion, Param4]}
  n58 = LinearInterpolate(
   A=TextureCoordinate()
   B=TextureCoordinate() {coordinateIndex=1}
   Alpha=ScalarParameter() {group=Erosion, parameterName=Erosion UV Channel, sortPriority=3}
  )
  n59 = Multiply(
   A=LinearInterpolate(
    A=@n58
    B=MaterialFunctionCall(XY=@n58) {materialFunction=&Swizzle}.YX
    Alpha=ScalarParameter() {group=Erosion, parameterName=Swizzle Erosion UVs, sortPriority=4}
   )
   B=AppendVector(
    A=ScalarParameter() {defaultValue=1, group=Erosion, parameterName=Erosion UV ScaleX, sortPriority=5}
    B=ScalarParameter() {defaultValue=1, group=Erosion, parameterName=Erosion UV ScaleY, sortPriority=6}
   )
  )
  n60 = AppendVector(A=@n57.Erosion ScrollX, B=@n57.Erosion ScrollX)
  n61 = ScalarParameter() {group=Erosion, parameterName=Erosion PanX, sortPriority=7}
  n62 = ScalarParameter() {group=Erosion, parameterName=Erosion PanY, sortPriority=8}
  n63 = TextureSampleParameter2D(UVs=Panner(Coordinate=Add(A=@n59, B=@n60), Speed=AppendVector(A=@n61, B=@n62))) {
   group=Erosion
   parameterName=Erosion Texture
   sortPriority=1
   texture=&T_Pickaxe_SoapPocket_NoiseGlow
  }
  n64 = ChannelMaskParameter(None=@n63.RGB) {group=Erosion, parameterName=Erosion Texture Channel, sortPriority=2}
  n65 = ScalarParameter() {defaultValue=1, group=Erosion, parameterName=Erosion Multiply, sortPriority=9}
  n66 = MaterialFunctionCall(
   Alpha=Saturate(None=Power(Base=@n56, Exp=@n57.Erosion) {constExponent=1})
   Min=Power(Base=Multiply(A=@n64.R, B=@n65), Exp=ScalarParameter() {defaultValue=1, group=Erosion, parameterName=Erosion Power, sortPriority=10})
  ) {materialFunction=&SmoothStep}
  n67 = ScalarParameter() {defaultValue=1, group=Erosion, parameterName=UseErosion, sortPriority=0}
  n68 = DepthFade(
   Opacity=Multiply(A=Multiply(A=@n47, B=@n54), B=LinearInterpolate(A=@n55, B=@n66.Result, Alpha=@n67))
   FadeDistance=ScalarParameter() {defaultValue=10, group=Fade Controls, parameterName=DepthFade, sortPriority=1}
  )
  n69 = MaterialFunctionCall(
   Fade Length=ScalarParameter() {desc=Camera Fade distance., group=Fade Controls, parameterName=Camera Fade Distance, sortPriority=2}
   Fade Offset=ScalarParameter() {desc=Camera fade offset., group=Fade Controls, parameterName=Camera Fade Offset, sortPriority=3}
  ) {materialFunction=&CameraDepthFade}
  n70 = ScalarParameter() {defaultValue=1, desc=Camera fade exponent., group=Fade Controls, parameterName=Camera Fade Exponent, sortPriority=4}
  n71 = LinearInterpolate(
   A=@n68
   B=Multiply(A=@n68, B=Power(Base=@n69.Result, Exp=@n70))
   Alpha=ScalarParameter() {group=Fade Controls, parameterName=UseCameraFade, sortPriority=-1}
  )
  n72 = MaterialFunctionCall("FadeSensitivity (optional)"=ScalarParameter() {group=Fade Controls, parameterName=Fade Sensitivity, sortPriority=5}) {materialFunction=&FadeBasedOnViewAngle}
  Multiply(
   A=@n44.A
   B=LinearInterpolate(A=@n71, B=Multiply(A=@n71, B=@n72.Result), Alpha=ScalarParameter() {group=Fade Controls, parameterName=UseAngleFade, sortPriority=-1})
  )
  n73 = DynamicParameter() {desc=Parameters set by Niagara, parameterIndex=2, paramNames [X,Y,Z,WPOIntensity]}
  n74 = Subtract(
   A=MaterialFunctionCall(X=@n73.X, Y=@n73.X, Z=@n73.X) {materialFunction=&MakeFloat3}.Result
   B=ObjectPositionWS()
  )
  n75 = ScalarParameter() {defaultValue=-300, group=WPO, parameterName=Clamp_min, sortPriority=4}
  n76 = ScalarParameter() {defaultValue=300, group=WPO, parameterName=Clamp_max, sortPriority=5}
  n77 = TextureCoordinate()
  n78 = Multiply(
   A=Power(Base=ComponentMask(None=@n77) {r=true}, Exp=ScalarParameter() {defaultValue=3, group=WPO, parameterName=WPOBendPower, sortPriority=3})
   B=@n73.WPOIntensity
  )
  n79 = LinearInterpolate(
   A=Constant()
   B=LinearInterpolate(B=Clamp(None=@n74, Min=@n75, Max=@n76) {maxDefault=1000, minDefault=-1000}, Alpha=LinearInterpolate(Alpha=@n78) {constB=0.8})
   Alpha=ScalarParameter() {group=WPO, parameterName=UseMesh, sortPriority=1}
  )
  n80 = Add(
   A=MaterialFunctionCall(
    WidthBase=ScalarParameter() {defaultValue=20, desc=Global spline width scalar., group=WPO, parameterName=Spline Width, sortPriority=6}
    WidthTip=ScalarParameter() {defaultValue=20, group=WPO, parameterName=Spline Tip Width, sortPriority=7}
   ) {materialFunction=&SplineThicken}.WorldPositionOffset
   B=@n79
  )
  n81 = ScalarParameter() {group=WPO, parameterName=UseSplineThicken, sortPriority=2}
  Reroute(None=StaticSwitchParameter(True=LinearInterpolate(A=@n79, B=@n80, Alpha=@n81), False=Constant()) {group=WPO, parameterName=UseWPO, sortPriority=0})
 ]
}
