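# Master translucent unlit material for Niagara ribbon effects, defaulting to an inky smoke trail texture.
# A panned texture, with UV channel and frequency driven by parameters and the Niagara bridge, feeds a chain of shared Material Attributes functions: skydive camera fade, component mask, color correct, colorize, opacity, depth fade and opacity bevel.
# A UV1 U remap fades opacity toward one ribbon end, and emissive can optionally be divided by eye adaptation for exposure control.
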Material /Game/Effects/Niagara/MasterMaterials/Ribbon/M_Ribbon_Master
graph {
 properties {
  bUseMaterialAttributes=true
  bUsedWithNiagaraRibbons=true
  blendMode=BLEND_Translucent
  pixelDepthOffsetMode=PDOM_Legacy
  shadingModel=MSM_Unlit
  usedShadingModels=Unlit
 }
 nodes [
  n1 = TextureCoordinate() {coordinateIndex=1}
  n2 = MaterialFunctionCall() {materialFunction=/Game/Effects/Niagara/MasterMaterials/Functions/MF_NFX_Bridge}
  n3 = AppendVector(A=@n2.UV1 U Freq, B=@n2.UV1 U Freq)
  n4 = TextureCoordinate()
  n5 = AppendVector(A=@n2.UV0 U Freq, B=@n2.UV0 U Freq)
  n6 = ScalarParameter() {group="2) BaseColor", parameterName="UV Channel (BaseColor)", sliderMax=1, sortPriority=1}
  n7 = TextureSampleParameter2D(UVs=MaterialFunctionCall!(
    UV Channel 2=Multiply(A=@n1, B=@n3)
    UV Channel 1=Multiply(A=@n4, B=@n5)
    In
    UV Channel=Clamp(None=@n6)
    U Freq Multiplier=ScalarParameter() {defaultValue=1, group="2) BaseColor", parameterName="U Freq (BaseColor)", sortPriority=2}
    V Freq Multiplier=ScalarParameter() {defaultValue=1, group="2) BaseColor", parameterName="V Freq (BaseColor)", sortPriority=3}
    Pan Speed U=ScalarParameter() {defaultValue=1, group="2) BaseColor", parameterName="Pan U Speed (BaseColor)", sortPriority=4}
    Pan Speed V=ScalarParameter() {group="2) BaseColor", parameterName="Pan V Speed (BaseColor)", sortPriority=5}
   ) {desc=The Niagara Bridge, materialFunction=/Game/Effects/Niagara/MasterMaterials/Functions/UVs/MF_NFX_TextureUVPanner}.Result) {
   desc=This texture provides the foundation for the other channels like opacity and emissive color.
   group="2) BaseColor"
   parameterName=Texture
   texture=/Game/Effects/Fort_Effects/Textures/T_InkySmokeTrail
  }
  n8 = MaterialFunctionCall!(In=MaterialFunctionCall!(Material Attributes=SetMaterialAttributes!(MaterialAttributes, Base Color=@n7.RGB, Emissive Color) {attributeSetTypes [#1, #2]}) {materialFunction=/Game/Effects/Niagara/MasterMaterials/Functions/V2/MF_Skydive_CameraFade_V2}."Opacitys and BaseColor (MA)") {
   desc=Controlled by the Material Component Mask Niagara Module.
   materialFunction=/Game/Effects/Niagara/MasterMaterials/Functions/MF_NFX_ComponentMask
  }
  n9 = MaterialFunctionCall!(In=@n8."BaseColor (MA) ") {
   desc=Controled throught the Material or Material ColorCorrect Niagara Module.
   materialFunction=/Game/Effects/Niagara/MasterMaterials/Functions/MF_Niagara_ColorCorrect
  }
  n10 = ScalarParameter() {
   desc=This parameter multiplies the texture by particle color and clamps it before going into the ldr range of the ldr to hdr function.
   group="5) Emissive"
   parameterName=Proper Particle HDR Coloring
   sliderMax=1
  }
  n11 = MaterialFunctionCall!(
   Distortion
   In=MaterialFunctionCall!(
    Override LDR imput with particle color=ScalarParameter() {group="2) BaseColor", parameterName=Override base color texture with particle color}
    In=@n9."BaseColor (MA)"
    Correct Coloring=@n10
   ) {materialFunction=/Game/Effects/Niagara/MasterMaterials/Functions/MF_FX_Colorize}."Color (MA)"
  ) {materialFunction=/Game/Effects/Niagara/MasterMaterials/Functions/MF_FX_Opacity}
  n12 = GetMaterialAttributes(None=MaterialFunctionCall!(In=MaterialFunctionCall!(In=@n11.Opacity) {materialFunction=/Game/Effects/Niagara/MasterMaterials/Functions/Opacity/FX_DepthFade}."Opacity (MA)") {materialFunction=/Game/Effects/Niagara/MasterMaterials/Functions/Opacity/MF_Parametric_OpacityBevel}."Opacitys and BaseColor (MA)") {attributeGetTypes [#3, #4]}
  n13 = MaterialFunctionCall(
   Input=ComponentMask(None=TextureCoordinate() {coordinateIndex=1}) {r=true}
   Input Low=Constant() {r=0.8}
   Input High=Constant() {r=0.9}
   Target Low=Constant() {r=1}
   Target High=Constant()
  ) {materialFunction=&RemapValueRange}
  n14 = GetMaterialAttributes(None=SetMaterialAttributes!(
    MaterialAttributes=@n12.MaterialAttributes
    Opacity=Multiply(A=@n12.Opacity, B=@n13.Result)
    Opacity Mask=Multiply(A=@n12.Opacity Mask, B=@n13.Result)
   ) {attributeSetTypes [#3, #4]}) {attributeGetTypes [#2]}
  n15 = LinearInterpolate(A=EyeAdaptation().EyeAdaptation, Alpha=Saturate(None=ScalarParameter() {parameterName=Eye Adaption Weight}))
  SetMaterialAttributes!(
   MaterialAttributes=@n14.MaterialAttributes
   Emissive Color=LinearInterpolate(
    A=@n14.Emissive Color
    B=Divide(A=@n14.Emissive Color, B=@n15)
    Alpha=ScalarParameter() {parameterName=Use Eye Adaption Exposure Control for Emissives}
   )
  ) {attributeSetTypes [#2]}
 ]
}
