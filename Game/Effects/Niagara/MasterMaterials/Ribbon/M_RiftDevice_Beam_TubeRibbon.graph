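# Additive unlit Niagara master material for the Rift Device's tube beam, usable on ribbons and mesh particles.
# Two panning chromatic caustic samples, warped along U and oversaturated, are max-blended with a magenta-to-cyan gradient driven by a sine pulse and axis-aligned Fresnel.
# Particle color tints the result, which gets HDR remapping on deferred shading.
# End-masked Fresnel forms an opacity term, depth-faded except at low quality.
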Material /Game/Effects/Niagara/MasterMaterials/Ribbon/M_RiftDevice_Beam_TubeRibbon
graph {
 properties {
  bAutomaticallySetUsageInEditor=false
  bUsedWithNiagaraMeshParticles=true
  bUsedWithNiagaraRibbons=true
  blendMode=BLEND_Additive
  shadingModel=MSM_Unlit
  usedShadingModels=Unlit
 }
 nodes [
  n1 = TextureCoordinate()
  n2 = ComponentMask(None=@n1) {r=true}
  n3 = Multiply(A=Power(Base=OneMinus(None=@n2)) {constExponent=10}, B=Constant2Vector() {r=-0.7})
  n4 = VectorParameter() {parameterName=TileAndPan, channelNames {a=Pan Y, b=Pan X, g=Tile Y, r=Tile X}, defaultValue=<rgba 3 2 1.25 0.5>}
  n5 = Multiply(A=@n1, B=AppendVector(A=@n4.Tile X, B=@n4.Tile X))
  n6 = AppendVector(A=@n4.Pan X, B=@n4.Pan X)
  n7 = Panner(Coordinate=@n5, Speed=@n6) {speedX=-0.5, speedY=0.2}
  n8 = Panner(Coordinate=Multiply(A=@n5, B=Constant2Vector() {g=1.2, r=1.1}), Speed=Multiply(A=@n6, B=Constant2Vector() {g=0.4, r=1.4})) {speedX=-0.5, speedY=0.2}
  n9 = Max(
   A=TextureSample(UVs=Add(A=@n3, B=@n7)) {texture=/Game/Effects/Fort_Effects/Textures/Tile/T_ShieldCaustics_Tile_Chromatic}.RGB
   B=TextureSample(UVs=Add(A=@n3, B=@n8)) {texture=&T_ShieldCaustics_Tile_Chromatic}.RGB
  )
  n10 = Constant() {r=-1}
  n11 = TextureCoordinate()
  n12 = ComponentMask(None=@n11) {g=true}
  n13 = LinearInterpolate(
   A=Constant3Vector() {constant=<rgba 0.8605 0.0566 1 1>}
   B=Constant3Vector() {constant=<rgba 0.0829 0.5726 1 1>}
   Alpha=Multiply(A=@n12, B=@n12)
  )
  n14 = Saturate(None=Multiply(A=ComponentMask(None=@n11) {r=true}) {constB=8})
  n15 = TextureCoordinate()
  n16 = Constant() {r=0.5}
  n17 = Power(Base=SphereMask(A=ComponentMask(None=@n15) {r=true}, B=@n16, Radius=@n16, Hardness=ScalarParameter() {defaultValue=0.94, parameterName=EndFadeHardness}))
  n18 = MaterialFunctionCall(
   Axis=MaterialFunctionCall() {materialFunction=&TangentBasis}.X U
   Exponent=Multiply(A=@n17, B=ScalarParameter() {defaultValue=5, parameterName=FresnelExp})
  ) {materialFunction=&AxisAlignedFresnel}
  n19 = Add(A=Multiply(A=NamedRerouteUsage(), B=@n14), B=Power(Base=Multiply(A=@n14, B=@n18.Fresnel)) {constExponent=5})
  n20 = Multiply(A=ParticleColor().RGB, B=Max(A=Desaturation(None=@n9, Fraction=@n10), B=Multiply(A=@n13, B=@n19)))
  n21 = Clamp(None=@n20)
  n22 = Constant() {r=0.5}
  MaterialFunctionCall(Emissive=ShadingPathSwitch(
    Default=@n20
    Deferred=MaterialFunctionCall(LDR Input=@n21, HDR Tint=@n20, Desaturate Luminance=@n22) {materialFunction=&DeriveHDRfromLDR}.RemappedRange
   )) {materialFunction=&MF_FX_EmissiveAdjustments}
  n23 = Power(Base=ComponentMask(None=TextureCoordinate() {uTiling=4}) {r=true}) {constExponent=0.6}
  n24 = Multiply(A=Time(), B=ScalarParameter() {defaultValue=3, parameterName=PulseSpeed})
  n25 = LinearInterpolate(Alpha=Sine(None=Add(A=@n23, B=@n24))) {constA=-0.5}
  NamedRerouteDeclaration(None=Power(Base=Saturate(None=@n25)) {constExponent=3}) {name=PulseMask, nodeColor=<rgba 0.4766 1 0 1>}
  n26 = Multiply(A=@n17, B=@n18.Fresnel)
  QualitySwitch(Default=DepthFade(Opacity=@n26, FadeDistance=ScalarParameter() {defaultValue=100, parameterName=DepthFadeDist}), Low=@n26)
 ]
}
