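# Translucent unlit master material for Niagara ribbon trails, with a soft, scrolling detail texture.
# A tiled trail texture pans by Time or a fixed U Offset, and its red channel is multiplied by particle alpha.
# A powered sine along U softens both ends before depth fade.
# A three-color gradient along U tints particle color, then passes through desaturation and a shared emissive adjustment function.
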Material /Game/Effects/Niagara/MasterMaterials/Ribbon/M_SoftDetailTrail
graph {
 properties {bUsedWithNiagaraRibbons=true, blendMode=BLEND_Translucent, pixelDepthOffsetMode=PDOM_Legacy, shadingModel=MSM_Unlit, usedShadingModels=Unlit}
 nodes [
  ComponentMask() {g=true, r=true}
  n1 = ParticleColor()
  n2 = TextureCoordinate()
  n3 = ComponentMask(None=VectorParameter() {parameterName=UV_Scale, defaultValue=<rgba 1.5 1 0 0>}.RGB) {g=true, r=true}
  n4 = Time()
  n5 = ScalarParameter() {parameterName=U Offset}
  n6 = ScalarParameter() {parameterName=Use Fixed timer}
  n7 = VectorParameter() {parameterName=Speed, defaultValue=<rgba 1.5 0 0 0>}
  n8 = TextureSampleParameter2D(UVs=Panner(Coordinate=Multiply(A=@n2, B=@n3), Time=LinearInterpolate(A=@n4, B=@n5, Alpha=@n6), Speed=ComponentMask(None=@n7.RGB) {g=true, r=true}) {speedX=1.5}) {group=Trail, parameterName=Base_Trail_texture, texture=/Game/Weapons/FORT_Melee/Pickaxe_Tower_Sentinel_Female/FX/T_TowerSentinel_trail}
  n9 = Power(Base=@n8.R) {constExponent=1}
  n10 = Sine(None=MaterialFunctionCall(Float2=TextureCoordinate()) {materialFunction=&BreakOutFloat2Components}.R) {period=2}
  DepthFade(
   Opacity=Multiply(A=Multiply(A=@n1.A, B=@n9), B=Power(Base=@n10))
   FadeDistance=ScalarParameter() {defaultValue=100, parameterName=FadeDistance}
  )
  n11 = TextureCoordinate()
  n12 = MaterialFunctionCall(
   A=VectorParameter() {parameterName=Color1, defaultValue=<rgba 0 0.8758 1 1>}.RGB
   B=VectorParameter() {parameterName=Color2, defaultValue=<rgba 0.629 0.0786 1 1>}.RGB
   C=VectorParameter() {parameterName=Color3, defaultValue=<rgba 0 0.1397 1 1>}.RGB
   Alpha=Multiply(
    A=MaterialFunctionCall(Float2=@n11) {materialFunction=&BreakOutFloat2Components}.R
    B=ScalarParameter() {defaultValue=0.5, group=Trail, parameterName=Gradient_intensity}
   ) {constB=0.5}
  ) {materialFunction=&3ColorBlend}
  MaterialFunctionCall(Emissive=Desaturation(None=Multiply(A=@n1.RGB, B=@n12.Result), Fraction=ScalarParameter() {parameterName=Des})) {materialFunction=&MF_FX_EmissiveAdjustments}
 ]
}
