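# Translucent unlit master material for water-style Niagara ribbons and mesh particles, colored by particle color.
# Opacity layers noise-distorted panning plasma with edge and end masks, then erodes it with caustics-noise dissolve offering hard or soft edges.
# Near-camera fade applies throughout; depth fade is skipped on mobile quality tiers.
# Niagara dynamic parameters drive dissolve and World Position Offset, combining sine waves with an optional clamped bend toward a supplied position.
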Material /Game/Effects/Niagara/MasterMaterials/Ribbon/M_WaterStyle_Ribbon
graph {
 properties {
  bUsedWithNiagaraMeshParticles=true
  bUsedWithNiagaraRibbons=true
  blendMode=BLEND_Translucent
  pixelDepthOffsetMode=PDOM_Legacy
  shadingModel=MSM_Unlit
  twoSided=true
  usedShadingModels=Unlit
 }
 nodes [
  n1 = ComponentMask(None=TextureCoordinate() {uTiling=0}) {g=true}
  n2 = LinearInterpolate(
   A=@n1
   B=OneMinus(None=@n1)
   Alpha=DynamicParameter() {parameterIndex=2, paramNames [Width Swizzle Direction, Param2, Param3, Param4]}.Width Swizzle Direction
  )
  AppendVector(A=Constant() {r=1}, B=@n2)
  Constant() {r=7}
  MaterialFunctionCall(Emissive=ParticleColor().RGB) {materialFunction=&MF_FX_EmissiveAdjustments}
  NamedRerouteDeclaration(None=Multiply(A=Time(), B=ScalarParameter() {defaultValue=1, parameterName=Pan Speed})) {name=Pan Time, nodeColor=<rgba 1 0.0938 0 1>}
  n3 = Add(A=NamedRerouteUsage(), B=DynamicParameter() {defaultValue=<rgba 1 1 70 1>, paramNames [WPO Offset, WPO, TrailRange, Dis]}.WPO Offset)
  n4 = Multiply(A=NamedRerouteUsage()) {constB=-0.4}
  n5 = Sine(None=ComponentMask(None=Add(A=@n3, B=@n4)) {r=true}) {period=0.7}
  n6 = Multiply(A=@n5, B=Constant3Vector() {constant=<rgba 0 0 0.5 0>})
  n7 = Multiply(A=LinearInterpolate(Alpha=@n5) {constA=1, constB=2}, B=Constant3Vector() {constant=<rgba 0 0.5 0 0>})
  n8 = Multiply(
   A=Multiply(
    A=ScalarParameter() {defaultValue=7, group=WPO, parameterName=WPO Strength}
    B=DynamicParameter() {defaultValue=<rgba 1 1 70 1>, paramNames [WPO Offset, WPO, TrailRange, Dis]}.WPO
   )
   B=Add(A=Transform(None=@n6), B=Transform(None=@n7))
  ) {constB=7}
  n9 = TextureCoordinate()
  n10 = Saturate(None=Multiply(A=ComponentMask(None=@n9) {r=true}) {constB=5})
  n11 = DynamicParameter() {desc=Parameters set by Niagara, parameterIndex=1, paramNames [X,Y,Z,WPOIntensity]}
  n12 = MaterialFunctionCall(X=@n11.X, Y=@n11.X, Z=@n11.X) {materialFunction=&MakeFloat3}
  n13 = ParticlePositionWS()
  n14 = TextureCoordinate()
  n15 = ComponentMask(None=@n14) {r=true}
  n16 = ScalarParameter() {defaultValue=2.5, group=WPO, parameterName=WPOBendPower}
  n17 = LinearInterpolate(
   B=Clamp(
    None=Subtract(A=@n12.Result, B=@n13)
    Min=ScalarParameter() {defaultValue=-100, group=WPO, parameterName=Clamp_min}
    Max=ScalarParameter() {defaultValue=100, group=WPO, parameterName=Clamp_max}
   ) {maxDefault=1000, minDefault=-1000}
   Alpha=Multiply(A=@n11.WPOIntensity, B=Power(Base=@n15, Exp=@n16))
  )
  n18 = Constant()
  NamedRerouteDeclaration(None=Add(A=Multiply(A=@n8, B=@n10), B=StaticSwitchParameter(True=@n17, False=@n18) {group=WPO, parameterName=Use_Interpolate WPO}) {constB=0}) {name=WPO, nodeColor=<rgba 1 0.6609 0 1>}
  n19 = ParticleColor()
  n20 = Sine(None=ComponentMask(None=TextureCoordinate()) {g=true}) {period=2}
  n21 = NamedRerouteUsage()
  n22 = TextureSample(UVs=Panner(Coordinate=@n21, Time=NamedRerouteUsage()) {speedX=-0.05, speedY=-0.1}) {samplerType=SAMPLERTYPE_Grayscale, texture=&LowResBlurredNoise}
  n23 = Panner(Coordinate=Add(A=@n21, B=Multiply(A=@n22.R) {constB=0.2}), Time=NamedRerouteUsage()) {speedX=-0.3, speedY=0.05}
  n24 = Multiply(A=Power(Base=TextureSample(UVs=@n23) {samplerType=SAMPLERTYPE_LinearGrayscale, texture=/Game/Effects/Fort_Effects/Emotes/UltraLight/Texture/T_Plasma_C_Greyscale}.R)) {constB=0.1}
  n25 = SphereMask(A=TextureCoordinate() {uTiling=0}, B=Constant2Vector() {g=0.5}) {attenuationRadius=1, hardnessPercent=0}
  n26 = Add(A=Power(Base=OneMinus(None=@n25)))
  n27 = Multiply(A=@n26) {constB=0.03}
  n28 = OneMinus(None=SphereMask(A=TextureCoordinate() {uTiling=0}, B=Constant2Vector() {g=0.5}) {attenuationRadius=1, hardnessPercent=20})
  n29 = Multiply(
   A=Saturate(None=Multiply(A=@n20) {constB=4})
   B=LinearInterpolate(A=Add(A=@n24, B=@n27), B=Saturate(None=@n26), Alpha=Power(Base=@n28) {constExponent=2.5}) {constA=0.05}
  )
  n30 = Multiply(A=ComponentMask(None=TextureCoordinate()) {r=true}, B=ScalarParameter() {defaultValue=20, parameterName=Lead Edge mask}) {constB=20}
  n31 = TextureCoordinate()
  n32 = Power(Base=OneMinus(None=ComponentMask(None=@n31) {r=true})) {constExponent=1.5}
  n33 = Multiply(A=Multiply(A=Multiply(A=@n19.A, B=@n29), B=Saturate(None=@n30)), B=Saturate(None=Multiply(A=@n32) {constB=5}))
  n34 = Add(
   A=NamedRerouteUsage()
   B=AppendVector(A=DynamicParameter() {defaultValue=<rgba 1 1 70 1>, paramNames [WPO Offset, WPO, TrailRange, Dis]}.WPO Offset, B=Constant())
  )
  n35 = NamedRerouteUsage()
  n36 = ScalarParameter() {parameterName=Dissolve mask Min}
  n37 = DynamicParameter() {defaultValue=<rgba 1 1 70 1>, paramNames [WPO Offset, WPO, TrailRange, Dis]}
  n38 = Power(Base=ComponentMask(None=TextureCoordinate()) {r=true}, Exp=ScalarParameter() {defaultValue=1.3, parameterName=Dissolve Falloff Exponent}) {constExponent=1.3}
  n39 = Power(
   Base=TextureSample(UVs=Panner(Coordinate=@n34, Time=@n35) {speedX=-0.5}) {samplerType=SAMPLERTYPE_LinearGrayscale, texture=/Game/Global/Textures/Default/Caustics/T_Nosie_Caustics_Grayscale}.R
   Exp=Saturate(None=LinearInterpolate(A=@n36, B=@n37.Dis, Alpha=@n38) {constA=0.1, constB=2})
  )
  n40 = Constant() {r=0.5}
  n41 = ScalarParameter() {parameterName=Soft Edge Width}
  n42 = Subtract(A=@n40, B=@n41)
  n43 = Add(A=@n40, B=@n41)
  n44 = Saturate(None=LinearInterpolate(
    A=Round(None=@n39)
    B=SmoothStep(Min=@n42, Max=@n43, Value=@n39) {constMax=0.6, constMin=0.4, constValue=0.3}
    Alpha=ScalarParameter() {parameterName=Hard or Soft Edge}
   ))
  n45 = Multiply(
   A=MaterialFunctionCall(Distance=ScalarParameter() {defaultValue=256, parameterName=Near Camera Fade}) {materialFunction=&Near_Camera_Fade}.Result
   B=Multiply(A=NamedRerouteUsage(), B=Multiply(A=@n33, B=@n44))
  )
  n46 = DepthFade(Opacity=@n45, FadeDistance=ScalarParameter() {defaultValue=10, parameterName=Fade Distance}) {fadeDistanceDefault=10}
  NamedRerouteDeclaration(None=MaterialFunctionCall!(
    Low _ PC Low and Med=@n46
    "High _ XB1, PS4"=@n46
    "Epic _ PC High, XSX, PS5"=@n46
    --------
    "Mobile Low _ PC Perf, Android"=@n45
    Mobile Med _ Switch=@n45
   ) {materialFunction=/Game/Global/MaterialFunctions/TechArt/MF_Fortnite_QualitySwitch_Scalar}.Result) {name=Opacity, nodeColor=<rgba 0.5469 1 0 1>}
  NamedRerouteDeclaration(None=Multiply(A=TextureCoordinate() {vTiling=0.3}, B=ScalarParameter() {defaultValue=1, parameterName=Texture Scale})) {name=Scaled TexCoord, nodeColor=<rgba 1 0 0.6667 1>}
  n47 = ScalarParameter() {group=Width Fade, parameterName=Width Fade Min}
  n48 = ScalarParameter() {defaultValue=1, group=Width Fade, parameterName=Width Fade Max}
  n49 = Power(Base=@n2, Exp=ScalarParameter() {defaultValue=1, group=Width Fade, parameterName=Width Fade Exponent})
  NamedRerouteDeclaration(None=LinearInterpolate(A=Constant() {r=1}, B=LinearInterpolate(A=@n47, B=@n48, Alpha=@n49), Alpha=ScalarParameter() {group=Width Fade, parameterName=Do Width Fade})) {name=Width Fade, nodeColor=<rgba 0 0.6953 1 1>}
  NamedRerouteUsage()
  NamedRerouteUsage()
  OneMinus(None=@n14)
  TextureCoordinate() {uTiling=0.8, vTiling=0.2}
  TextureCoordinate() {vTiling=0.5}
  TextureCoordinate()
  Time()
 ]
}
