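# Master additive, unlit, two-sided material for Niagara ribbon trails, drawing a thin scrolling, distorted wave trail.
# A tiled grayscale texture pans on UV1, offset by a panning distortion normal map, then gets a power curve.
# A static switch tints it with a two-color gradient or an HDR color, then desaturation and particle color apply.
# Opacity squares the texture, depth-fades it except on mobile, and multiplies UV0 lengthwise masks and particle alpha.
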Material /Game/Effects/Niagara/MasterMaterials/Ribbon/M_WaveThin_Trail
graph {
 properties {
  bUsedWithNiagaraRibbons=true
  blendMode=BLEND_Additive
  pixelDepthOffsetMode=PDOM_Legacy
  shadingModel=MSM_Unlit
  twoSided=true
  usedShadingModels=Unlit
 }
 nodes [
  Constant() {r=0.3}
  n1 = VectorParameter() {parameterName=Base Color A, defaultValue=<rgba 0.5 0.5 0.5 1>}
  n2 = VectorParameter() {parameterName=Base Color B, defaultValue=<rgba 5 5 5 1>}
  n3 = Power(
   Base=MaterialFunctionCall(Float2=TextureCoordinate() {coordinateIndex=1}) {materialFunction=&BreakOutFloat2Components}.R
   Exp=ScalarParameter() {defaultValue=2, parameterName=Color Power}
  )
  n4 = VectorParameter() {parameterName=Main color, defaultValue=<rgba 1 1 1 1>}
  n5 = Clamp(None=@n4.RGB)
  n6 = VectorParameter() {parameterName=Base UV, channelNames {a="", b=Offset V, g=Speed U, r=Tile U}, defaultValue=<rgba 1 2 0.2 0>}
  n7 = Constant() {r=1}
  n8 = Panner(
   Coordinate=Multiply(A=TextureCoordinate() {coordinateIndex=1}, B=AppendVector(A=@n6.Tile U, B=@n7))
   Speed=AppendVector(A=@n6.Speed U, B=Constant())
  ) {speedX=1}
  n9 = Panner(Coordinate=TextureCoordinate() {coordinateIndex=1, uTiling=0, vTiling=0}, Time=@n6.Offset V) {speedY=1}
  n10 = TextureCoordinate() {coordinateIndex=1, uTiling=-0.35, vTiling=0.5}
  n11 = AppendVector(A=ScalarParameter() {defaultValue=-0.8, parameterName=Distort Pan}, B=Constant())
  n12 = ComponentMask(None=TextureSample(UVs=Panner(Coordinate=@n10, Speed=@n11) {speedX=-0.8}) {samplerType=SAMPLERTYPE_Normal, texture=&T_Weap_Distort_N_02}.RGB) {g=true, r=true}
  n13 = ScalarParameter() {defaultValue=0.28, parameterName=Distort_Intensity}
  n14 = TextureSampleParameter2D(UVs=Add(A=Add(A=@n8, B=@n9) {constB=0}, B=Multiply(A=@n12, B=@n13) {constB=0.28})) {parameterName=Base Texture, samplerType=SAMPLERTYPE_Grayscale, texture=&T_Smoke_Trail_Soft}
  n15 = Power(Base=@n14.RGB, Exp=ScalarParameter() {defaultValue=1.5, parameterName=Texture power})
  n16 = Multiply(
   A=StaticSwitchParameter(
    True=LinearInterpolate(A=@n1.RGB, B=@n2.RGB, Alpha=@n3)
    False=MaterialFunctionCall(LDR Input=@n5, HDR Tint=@n4.RGB) {materialFunction=&DeriveHDRfromLDR}.RemappedRange
   ) {defaultValue=true, parameterName=Use Rainbow color}
   B=@n15
  )
  n17 = ScalarParameter() {parameterName=Desaturation Color}
  n18 = ParticleColor()
  MaterialFunctionCall(Emissive=Multiply(A=Desaturation(None=@n16, Fraction=@n17), B=@n18.RGB)) {materialFunction=&MF_FX_EmissiveAdjustments}
  n19 = Multiply(A=@n15, B=@n15)
  n20 = DepthFade(Opacity=@n19, FadeDistance=ScalarParameter() {defaultValue=100, parameterName=DepthFade-Standard_Distance}) {fadeDistanceDefault=20}
  n21 = OneMinus(None=ComponentMask(None=TextureCoordinate()) {r=true})
  n22 = TextureCoordinate()
  n23 = Power(
   Base=Multiply(A=ComponentMask(None=@n22) {r=true}, B=ScalarParameter() {defaultValue=0.5, parameterName=Holiday Cookies Mask Mult}) {constB=5}
   Exp=ScalarParameter() {defaultValue=0.5, parameterName=Holiday Cookies Mask Power}
  ) {constExponent=0.5}
  Multiply(
   A=@n18.A
   B=Multiply(A=ShadingPathSwitch(Default=@n20, Deferred=@n20, Mobile=@n19) {desc=Migrated from FeatureLevelSwitch}, B=Multiply(A=@n21, B=@n23))
  )
 ]
}
