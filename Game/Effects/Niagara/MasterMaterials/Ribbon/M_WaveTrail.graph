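# Translucent unlit Niagara master material for ribbon and mesh particle trails, drawing a scrolling, distorted smoke trail tinted by particle color.
# A panning normal texture distorts the UVs of a rainbow detail, noise and smoke texture, with dynamic OffsetX shifting the smoke.
# Desaturated detail, noise and smoke form optional emissive color through MF_FX_EmissiveAdjustments.
# Smoke, U-axis end fades, optional noise and depth fade (skipped on mobile) drive alpha into DitherTemporalAA.
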Material /Game/Effects/Niagara/MasterMaterials/Ribbon/M_WaveTrail
graph {
 properties {
  bCastRayTracedShadows=false
  bUsedWithNiagaraMeshParticles=true
  bUsedWithNiagaraRibbons=true
  blendMode=BLEND_Translucent
  pixelDepthOffsetMode=PDOM_Legacy
  shadingModel=MSM_Unlit
  usedShadingModels=Unlit
 }
 nodes [
  n1 = ScalarParameter() {defaultValue=0.5, group=Distortion, parameterName=Distortion Detail Intensity, sortPriority=2}
  n2 = TextureCoordinate() {coordinateIndex=1}
  n3 = VectorParameter() {group=Distortion, parameterName=Distortion UV, sortPriority=1, defaultValue=<rgba 0.5 0.5 0 0>}
  n4 = AppendVector(
   A=ScalarParameter() {defaultValue=1, group=Noise, parameterName=Global Scale U}
   B=ScalarParameter() {defaultValue=1, group=Noise, parameterName=Global Scale V}
  )
  n5 = Multiply(A=AppendVector(A=@n3.R, B=@n3.R), B=@n4)
  n6 = TextureSampleParameter2D(UVs=Panner(Coordinate=Multiply(A=@n2, B=@n5), Speed=AppendVector(A=@n3.B, B=@n3.B))) {
   group=Distortion
   parameterName="Distortion  Tex"
   samplerType=SAMPLERTYPE_Normal
   sortPriority=0
   texture=&T_Weap_Distort_N_02
  }
  n7 = ComponentMask(None=@n6.RGB) {g=true, r=true}
  n8 = VectorParameter() {group=Emissive, parameterName=Emissive Detail UV, sortPriority=1, defaultValue=<rgba 0.5 0.7 0.5 0>}
  n9 = ScalarParameter() {defaultValue=1, group=Noise, parameterName=Global Scale U}
  n10 = ScalarParameter() {defaultValue=1, group=Noise, parameterName=Global Scale V}
  n11 = Multiply(A=TextureCoordinate(), B=Multiply(A=AppendVector(A=@n8.R, B=@n8.R), B=AppendVector(A=@n9, B=@n10)))
  n12 = AppendVector(A=@n8.B, B=@n8.B)
  n13 = TextureSample(UVs=Add(A=Multiply(A=@n1, B=@n7), B=Panner(Coordinate=@n11, Speed=@n12))) {texture=&T_Tile_RainbowColor}
  n14 = ScalarParameter() {defaultValue=1, group=Emissive, parameterName=Emissive Detail Intensity}
  n15 = Multiply(A=@n7, B=ScalarParameter() {defaultValue=0.05, group=Distortion, parameterName=Distortion Intensity, sortPriority=2})
  n16 = VectorParameter() {group=Noise, parameterName=Noise A UV, sortPriority=6, defaultValue=<rgba 0.5 1 0.5 0>}
  n17 = Multiply(A=TextureCoordinate() {coordinateIndex=1}, B=Multiply(A=AppendVector(A=@n16.R, B=@n16.R), B=@n4))
  n18 = AppendVector(A=@n16.B, B=@n16.B)
  n19 = TextureSampleParameter2D(UVs=Add(A=@n15, B=Panner(Coordinate=@n17, Speed=@n18))) {group=Noise, parameterName=Noise Tex, samplerType=SAMPLERTYPE_LinearColor, texture=/Game/Global/Textures/Default/Noise/T_Noise01}
  n20 = ChannelMaskParameter(None=@n19.RGB) {parameterName=NoiseChannel}
  n21 = Multiply(
   A=Desaturation(None=Multiply(A=@n13.RGB, B=@n14), Fraction=ScalarParameter() {defaultValue=0.4, group=Emissive, parameterName=Desaturation Detail Color})
   B=LinearInterpolate(
    A=ScalarParameter() {defaultValue=0.8, group=Noise, parameterName=Noise A Dark}
    B=ScalarParameter() {defaultValue=1, group=Noise, parameterName=Noise A Light}
    Alpha=Power(Base=@n20.R)
   )
  )
  n22 = DynamicParameter() {defaultValue=<rgba 0 0 1 1>, paramNames [OffsetX,OffsetY,Param3,Param4]}
  n23 = VectorParameter() {group=Opacity, parameterName=SmokeTrail UV, sortPriority=1, defaultValue=<rgba 1 1 2 0>}
  n24 = ScalarParameter() {defaultValue=1, group=Noise, parameterName=Global Scale U}
  n25 = ScalarParameter() {defaultValue=1, group=Noise, parameterName=Global Scale V}
  n26 = Multiply(
   A=Add(A=AppendVector(A=@n22.OffsetX, B=@n22.OffsetX), B=TextureCoordinate() {coordinateIndex=1})
   B=Multiply(A=AppendVector(A=@n23.R, B=@n23.R), B=AppendVector(A=@n24, B=@n25))
  )
  n27 = AppendVector(A=@n23.B, B=@n23.B)
  n28 = TextureSampleParameter2D(UVs=Add(A=@n15, B=Panner(Coordinate=@n26, Speed=@n27))) {
   group=Opacity
   parameterName=SmokeTrail Tex
   samplerType=SAMPLERTYPE_Grayscale
   texture=&T_Smoke_Trail_Soft
  }
  n29 = ParticleColor()
  n30 = Multiply(
   A=Multiply(A=Multiply(A=@n21, B=@n28.R), B=VectorParameter() {group=Emissive, parameterName=Emissive Color, defaultValue=<rgba 1 1 1 1>}.RGB)
   B=@n29.RGB
  )
  MaterialFunctionCall(Emissive=If(A=ScalarParameter() {defaultValue=1, parameterName=UseDetailColor}, "A > B"=@n30, "A == B"=@n30, "A < B"=@n29.RGB) {constB=0.9}) {materialFunction=&MF_FX_EmissiveAdjustments}
  n31 = TextureCoordinate()
  n32 = ComponentMask(None=@n31) {r=true}
  n33 = Multiply(A=Power(Base=OneMinus(None=@n32)) {constExponent=1}) {constB=2}
  n34 = Multiply(A=ComponentMask(None=@n31) {r=true}) {constB=20}
  n35 = Multiply(A=@n28.R, B=Multiply(A=Saturate(None=@n33), B=Saturate(None=@n34)))
  n36 = Multiply(
   A=If(A=ScalarParameter() {parameterName=UseNoiseOnSmokeTrailText}, "A > B"=Multiply(A=@n20.R, B=@n35), "A == B"=@n35, "A < B"=@n35) {constB=0.9}
   B=ScalarParameter() {defaultValue=1, group=Opacity, parameterName=Opactiy Mult}
  )
  n37 = DepthFade(Opacity=@n36, FadeDistance=ScalarParameter() {defaultValue=40, group=Opacity, parameterName=DepthFadeDistance})
  MaterialFunctionCall(
   Alpha Threshold=Multiply(A=@n29.A, B=ShadingPathSwitch(Default=@n37, Mobile=@n36) {desc=Migrated from FeatureLevelSwitch})
   Random=ScalarParameter() {group=Opacity, parameterName=MaskDither}
  ) {materialFunction=&DitherTemporalAA}
 ]
}
